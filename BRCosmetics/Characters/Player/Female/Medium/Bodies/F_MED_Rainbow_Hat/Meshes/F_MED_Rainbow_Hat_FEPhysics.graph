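# Physics asset for the Rainbow_Hat female medium character body, setting up secondary motion for sleeves, socks, two grenades and six plush attachments.
# All bodies are capsules, including large thigh capsules and small ones at most chain roots.
# Sleeve pieces hang from upper-arm twist bones, and sock pieces from base bones whose thigh joints swing and twist freely.
# Plushes and grenades form separate angle-limited chains, with ears branching on plush C and F.
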PhysicsAsset /BRCosmetics/Characters/Player/Female/Medium/Bodies/F_MED_Rainbow_Hat/Meshes/F_MED_Rainbow_Hat_FEPhysics
bodies {
 dyn_sleeve_01_l [{shapeType=Capsule, length=2, radius=2, center=<2.208 0 0>, rotation=<rot 90 0 0>}]
 dyn_sleeve_01_r [{shapeType=Capsule, length=2, radius=2, center=<-1 0 0>, rotation=<rot 90 0 0>}]
 dyn_sleeve_02_l [{shapeType=Capsule, length=2, radius=2.25, center=<0.9969 0.0001 0>, rotation=<rot 90 0 0>}]
 dyn_sleeve_02_r [{shapeType=Capsule, length=2, radius=2.25, center=<-2 0 0>, rotation=<rot 90 0 0>}]
 dyn_sleeve_03_l [{shapeType=Capsule, length=3, radius=2, center=<1.9687 0 0>, rotation=<rot 90 0 0>}]
 dyn_sleeve_03_r [{shapeType=Capsule, length=3, radius=2, center=<-2 0 0>, rotation=<rot 90 0 0>}]
 dyn_sleeve_04_l [{shapeType=Capsule, length=1.75, radius=2.25, center=<1.1009 0.0001 0>, rotation=<rot 90 0 0>}]
 dyn_sleeve_04_r [{shapeType=Capsule, length=1.75, radius=2.25, center=<-0.78 0 0>, rotation=<rot 90 0 0>}]
 dyn_socksA_l [{shapeType=Capsule, length=3, radius=1.5, center=<0.8 0 0>, rotation=<rot 90 0 0>}]
 dyn_socksA_r [{shapeType=Capsule, length=3, radius=1.5, center=<1 0 0>, rotation=<rot 90 0 0>}]
 dyn_socksB_l [{shapeType=Capsule, length=3, radius=1.5, center=<1.25 0 0>, rotation=<rot 90 0 0>}]
 dyn_socksB_r [{shapeType=Capsule, length=3, radius=1.5, center=<0.8 0 0>, rotation=<rot 90 0 0>}]
 dyn_socksC_l [{shapeType=Capsule, length=3, radius=1.5, center=<1.25 0 0>, rotation=<rot 90 0 0>}]
 dyn_socksC_r [{shapeType=Capsule, length=3, radius=1.5, center=<0.8988 0 0>, rotation=<rot 90 0 0>}]
 dyn_socksD_l [{shapeType=Capsule, length=3, radius=1.5, center=<0.8988 0 0>, rotation=<rot 90 0 0>}]
 dyn_socksD_r [{shapeType=Capsule, length=1.5, radius=1.5, center=<1.25 0 0>, rotation=<rot 90 0 0>}]
 dyn_socksE_l [{shapeType=Capsule, length=3, radius=1.5, center=<1 0 0>, rotation=<rot 90 0 0>}]
 dyn_socksE_r [{shapeType=Capsule, length=1.5, radius=1.5, center=<1.25 0 0>, rotation=<rot 90 0 0>}]
 dyn_socksF_r [{shapeType=Capsule, length=3, radius=1.5, center=<1.25 0 0>, rotation=<rot 90 0 0>}]
 dyn_socksG_r [{shapeType=Capsule, length=3, radius=1.5, center=<0.8988 0 0>, rotation=<rot 90 0 0>}]
 dyn_socksH_r [{shapeType=Capsule, length=1.5, radius=1.5, center=<1.25 0 0>, rotation=<rot 90 0 0>}]
 dyn_socks_base_l [{shapeType=Capsule, length=0.505, radius=0.505, rotation=<rot 0 -90 0>}]
 dyn_socks_base_r [{shapeType=Capsule, length=0.505, radius=0.505, rotation=<rot 0 -90 0>}]
 grenadeA_01 [{shapeType=Capsule, length=0.505, radius=0.505, rotation=<rot 0 -90 0>}]
 grenadeA_02 [{shapeType=Capsule, length=6.3466, radius=3.2723, center=<0.7 -1.5231 2.8941>, rotation=<rot 0 89.3644 -88.3971>}]
 grenadeB_01 [{shapeType=Capsule, length=0.505, radius=0.505, rotation=<rot 0 -90 0>}]
 grenadeB_02 [{shapeType=Capsule, length=8.3466, radius=4.0202, center=<2.8804 -1.7838 2.4639>, rotation=<rot 0 89.3344 -89.9403>}]
 plushA_01 [{shapeType=Capsule, length=0.505, radius=0.505, rotation=<rot 0 -90 0>}]
 plushA_02 [{shapeType=Capsule, length=1, radius=1}]
 plushA_03 [{shapeType=Capsule, length=3.6114, radius=2.9503, center=<3.5928 -0.1743 1.9744>, rotation=<rot 0 92.2717 -80.2765>}]
 plushB_01 [{shapeType=Capsule, length=0.505, radius=0.505, rotation=<rot 0 -90 0>}]
 plushB_02 [{shapeType=Capsule, length=1, radius=1}]
 plushB_03 [{shapeType=Capsule, length=3.25, radius=4.75, center=<7.3923 0 2.8844>, rotation=<rot 82.9133 179.9998 179.9998>}]
 plushC_01 [{shapeType=Capsule, length=0.505, radius=0.505, rotation=<rot 0 -90 0>}]
 plushC_02 [{shapeType=Capsule, length=2, radius=4, center=<1.5676 0.0003 3.2205>, rotation=<rot 90 0 0>}]
 plushC_03 [
  {shapeType=Capsule, length=2, radius=4, center=<1.2204 0 1.449>, rotation=<rot 61.155 0 0>}
  {shapeType=Capsule, length=12, radius=6.25, center=<9.1378 0.0003 -6.5875>, rotation=<rot 39.9057 -0.0001 0>}
 ]
 plushC_ear_01_l [{shapeType=Capsule, length=2.5, radius=1.5, center=<1.5 0 0>, rotation=<rot 90 0 0>}]
 plushC_ear_01_r [{shapeType=Capsule, length=2, radius=2, center=<2 0 0>, rotation=<rot 90 0 0>}]
 plushD_01 [{shapeType=Capsule, length=0.505, radius=0.505, rotation=<rot 0 -90 0>}]
 plushD_02 [{shapeType=Capsule, length=1, radius=1}]
 plushD_03 [{shapeType=Capsule, length=2, radius=4, center=<5.1846 0.0003 2.6748>, rotation=<rot 90 0 0>}]
 plushD_04 [
  {shapeType=Capsule, length=1.25, radius=4.75, center=<1.2204 0 1.449>, rotation=<rot 88.8778 0.0001 0.0001>}
  {shapeType=Capsule, length=7.75, radius=5.5, center=<14.3008 0 -1.8892>, rotation=<rot -71.6078 -179.9998 179.9999>}
 ]
 plushE_01 [{shapeType=Capsule, length=0.505, radius=0.505, rotation=<rot 0 -90 0>}]
 plushE_02 [{shapeType=Capsule, length=2.6114, radius=3.9503, center=<3.5928 -0.1743 1.9744>, rotation=<rot 0 92.2717 -80.2765>}]
 plushF_01 [{shapeType=Capsule, length=0.505, radius=0.505, rotation=<rot 0 -90 0>}]
 plushF_02 [{shapeType=Capsule, length=3.25, radius=1.25, center=<1 0 -0.0001>, rotation=<rot 90 0 0>}]
 plushF_03 [{shapeType=Capsule, length=5.25, radius=3.75, center=<5.1143 -0.1667 1.5742>, rotation=<rot 77.239 -179.9996 -179.9996>}]
 plushF_04 [
  {shapeType=Capsule, length=3.6791, radius=5.2451, center=<3.0341 -0.0232 0.9513>, rotation=<rot 0 6.1521 -87.2135>}
  {shapeType=Capsule, length=14, radius=9.5, center=<24.0191 -0.0001 1.9075>, rotation=<rot 89.0874 -179.9991 -179.999>}
 ]
 plushF_ear_01_l [{shapeType=Capsule, length=1.25, radius=0.75, center=<1 0 0>, rotation=<rot 90 0 0>}]
 plushF_ear_01_r [{shapeType=Capsule, length=1.5, radius=1.5, center=<1.2 0 0>, rotation=<rot 90 0 0>}]
 thigh_l [{shapeType=Capsule, length=27, radius=7.75, center=<-17.2424 -0.0769 0>, rotation=<rot 90 -2.9815 -5.901>}]
 thigh_r [{shapeType=Capsule, length=27, radius=7.75, center=<17.2424 -0.0769 0>, rotation=<rot 90 -2.9815 -5.901>}]
 upperarm_twist_02_l [{shapeType=Capsule, length=0.505, radius=0.505, rotation=<rot 0 -90 0>}]
 upperarm_twist_02_r [{shapeType=Capsule, length=0.505, radius=0.505, rotation=<rot 0 -90 0>}]
}
constraints [
 {bone1Name=dyn_socksA_l, bone2Name=dyn_socks_base_l, swing1LimitDegrees=15, swing2LimitDegrees=20, twistLimitDegrees=2}
 {bone1Name=dyn_socksB_l, bone2Name=dyn_socks_base_l, swing1LimitDegrees=15, swing2LimitDegrees=10, twistLimitDegrees=2}
 {bone1Name=dyn_socksC_l, bone2Name=dyn_socks_base_l, swing1LimitDegrees=15, swing2LimitDegrees=20, twistLimitDegrees=2}
 {bone1Name=dyn_socksD_l, bone2Name=dyn_socks_base_l, swing1LimitDegrees=15, swing2LimitDegrees=20, twistLimitDegrees=2}
 {bone1Name=dyn_socksE_l, bone2Name=dyn_socks_base_l, swing1LimitDegrees=15, swing2LimitDegrees=20, twistLimitDegrees=2}
 {bone1Name=dyn_socksA_r, bone2Name=dyn_socks_base_r, swing1LimitDegrees=15, swing2LimitDegrees=15, twistLimitDegrees=3}
 {bone1Name=dyn_socksB_r, bone2Name=dyn_socks_base_r, swing1LimitDegrees=15, swing2LimitDegrees=15, twistLimitDegrees=3}
 {bone1Name=dyn_socksC_r, bone2Name=dyn_socks_base_r, swing1LimitDegrees=15, swing2LimitDegrees=20, twistLimitDegrees=2}
 {bone1Name=dyn_socksD_r, bone2Name=dyn_socks_base_r, swing1LimitDegrees=15, swing2LimitDegrees=15, twistLimitDegrees=3}
 {bone1Name=dyn_socksE_r, bone2Name=dyn_socks_base_r, swing1LimitDegrees=15, swing2LimitDegrees=15, twistLimitDegrees=3}
 {bone1Name=dyn_socksF_r, bone2Name=dyn_socks_base_r, swing1LimitDegrees=15, swing2LimitDegrees=20, twistLimitDegrees=2}
 {bone1Name=dyn_socksG_r, bone2Name=dyn_socks_base_r, swing1LimitDegrees=15, swing2LimitDegrees=15, twistLimitDegrees=3}
 {bone1Name=dyn_socksH_r, bone2Name=dyn_socks_base_r, swing1LimitDegrees=15, swing2LimitDegrees=15, twistLimitDegrees=3}
 {bone1Name=plushA_02, bone2Name=plushA_01, swing1LimitDegrees=45, swing2LimitDegrees=45, twistLimitDegrees=8}
 {bone1Name=plushB_02, bone2Name=plushB_01, swing1LimitDegrees=27.8571, swing2LimitDegrees=41.5714, twistLimitDegrees=8}
 {
  bone1Name=dyn_socks_base_l
  bone2Name=thigh_l
  swing1LimitDegrees=45
  swing1Motion=Free
  swing2LimitDegrees=45
  swing2Motion=Free
  twistLimitDegrees=45
  twistMotion=Free
 }
 {
  bone1Name=dyn_socks_base_r
  bone2Name=thigh_r
  swing1LimitDegrees=45
  swing1Motion=Free
  swing2LimitDegrees=45
  swing2Motion=Free
  twistLimitDegrees=45
  twistMotion=Free
 }
 {bone1Name=plushC_02, bone2Name=plushC_01, swing1LimitDegrees=45, swing2LimitDegrees=14.7143, twistLimitDegrees=5}
 {bone1Name=plushC_03, bone2Name=plushC_02, swing1LimitDegrees=8, swing2LimitDegrees=15, twistLimitDegrees=5}
 {bone1Name=plushC_ear_01_r, bone2Name=plushC_02, swing1LimitDegrees=10, swing2LimitDegrees=15, twistLimitDegrees=3}
 {bone1Name=plushC_ear_01_l, bone2Name=plushC_02, swing1LimitDegrees=10, swing2LimitDegrees=15, twistLimitDegrees=3}
 {bone1Name=plushA_03, bone2Name=plushA_02, swing1LimitDegrees=45, swing2LimitDegrees=45, twistLimitDegrees=8}
 {bone1Name=plushB_03, bone2Name=plushB_02, swing1LimitDegrees=27.8571, swing2LimitDegrees=41.5714, twistLimitDegrees=8}
 {bone1Name=plushD_02, bone2Name=plushD_01, swing1LimitDegrees=45, swing2LimitDegrees=15, twistLimitDegrees=5}
 {bone1Name=plushD_03, bone2Name=plushD_02, swing1LimitDegrees=45, swing2LimitDegrees=25, twistLimitDegrees=5}
 {bone1Name=plushD_04, bone2Name=plushD_03, swing1LimitDegrees=8, swing2LimitDegrees=15, twistLimitDegrees=5}
 {bone1Name=plushE_02, bone2Name=plushE_01, swing1LimitDegrees=30, swing2LimitDegrees=30, twistLimitDegrees=4}
 {bone1Name=plushF_02, bone2Name=plushF_01, swing1LimitDegrees=15, swing2LimitDegrees=15, twistLimitDegrees=3}
 {bone1Name=plushF_03, bone2Name=plushF_02, swing1LimitDegrees=15, swing2LimitDegrees=15, twistLimitDegrees=5}
 {bone1Name=plushF_04, bone2Name=plushF_03, swing1LimitDegrees=15, swing2LimitDegrees=20, twistLimitDegrees=5}
 {bone1Name=plushF_ear_01_l, bone2Name=plushF_03, swing1LimitDegrees=13, swing2LimitDegrees=15, twistLimitDegrees=3}
 {bone1Name=plushF_ear_01_r, bone2Name=plushF_03, swing1LimitDegrees=12, swing2LimitDegrees=12, twistLimitDegrees=3}
 {bone1Name=grenadeA_02, bone2Name=grenadeA_01, swing1LimitDegrees=25, swing2LimitDegrees=20, twistLimitDegrees=3}
 {bone1Name=grenadeB_02, bone2Name=grenadeB_01, swing1LimitDegrees=15, swing2LimitDegrees=18, twistLimitDegrees=3}
 {bone1Name=dyn_sleeve_01_r, bone2Name=upperarm_twist_02_r, swing1LimitDegrees=20, swing2LimitDegrees=25, twistLimitDegrees=5}
 {bone1Name=dyn_sleeve_02_r, bone2Name=upperarm_twist_02_r, swing1LimitDegrees=20, swing2LimitDegrees=25, twistLimitDegrees=5}
 {bone1Name=dyn_sleeve_03_r, bone2Name=upperarm_twist_02_r, swing1LimitDegrees=20, swing2LimitDegrees=25, twistLimitDegrees=5}
 {bone1Name=dyn_sleeve_04_r, bone2Name=upperarm_twist_02_r, swing1LimitDegrees=20, swing2LimitDegrees=25, twistLimitDegrees=5}
 {bone1Name=dyn_sleeve_01_l, bone2Name=upperarm_twist_02_l, swing1LimitDegrees=20, swing2LimitDegrees=25, twistLimitDegrees=5}
 {bone1Name=dyn_sleeve_02_l, bone2Name=upperarm_twist_02_l, swing1LimitDegrees=20, swing2LimitDegrees=25, twistLimitDegrees=5}
 {bone1Name=dyn_sleeve_03_l, bone2Name=upperarm_twist_02_l, swing1LimitDegrees=20, swing2LimitDegrees=25, twistLimitDegrees=5}
 {bone1Name=dyn_sleeve_04_l, bone2Name=upperarm_twist_02_l, swing1LimitDegrees=20, swing2LimitDegrees=25, twistLimitDegrees=5}
]
