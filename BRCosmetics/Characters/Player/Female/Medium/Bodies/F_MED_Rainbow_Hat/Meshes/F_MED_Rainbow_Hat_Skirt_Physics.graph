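# Physics asset for the skirt of the female medium Rainbow Hat body cosmetic, defining capsule bodies and joints for its dynamic shirt and frill bones.
# Back and side chains link small parent capsules through driver bones to end bones, while the two front chains and the single left frill use two-bone links.
# Joints carry modest swing and tight twist limits, except the back-left end joint, whose swing and twist are set Free.
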PhysicsAsset /BRCosmetics/Characters/Player/Female/Medium/Bodies/F_MED_Rainbow_Hat/Meshes/F_MED_Rainbow_Hat_Skirt_Physics
bodies {
 dyn_driver_shirt_bck [{shapeType=Capsule, length=3, radius=3.255, center=<2 0 0>, rotation=<rot 90 0 0>}]
 dyn_driver_shirt_bck_l [{shapeType=Capsule, length=3, radius=3.255, center=<2 0 0>, rotation=<rot 90 0 0>}]
 dyn_driver_shirt_bck_r [{shapeType=Capsule, length=3, radius=3.255, center=<2 0 0>, rotation=<rot 90 0 0>}]
 dyn_driver_shirt_l [{shapeType=Capsule, length=3, radius=3.255, center=<2 0 0>, rotation=<rot 90 0 0>}]
 dyn_driver_shirt_r [{shapeType=Capsule, length=3, radius=3.255, center=<2 0 0>, rotation=<rot 90 0 0>}]
 dyn_frill_01_l [{shapeType=Capsule, length=2.25, radius=2.255, center=<2 0 0>, rotation=<rot 90 0 0>}]
 dyn_parent_frill_l [{shapeType=Capsule, length=0.505, radius=0.505, rotation=<rot 0 -90 0>}]
 dyn_parent_shirt_bck [{shapeType=Capsule, length=0.505, radius=0.505, rotation=<rot 0 -90 0>}]
 dyn_parent_shirt_bck_l [{shapeType=Capsule, length=0.505, radius=0.505, rotation=<rot 0 -90 0>}]
 dyn_parent_shirt_bck_r [{shapeType=Capsule, length=0.505, radius=0.505, rotation=<rot 0 -90 0>}]
 dyn_parent_shirt_l [{shapeType=Capsule, length=0.505, radius=0.505, rotation=<rot 0 -90 0>}]
 dyn_parent_shirt_r [{shapeType=Capsule, length=0.505, radius=0.505, rotation=<rot 0 -90 0>}]
 dyn_shirt_01_l [{shapeType=Capsule, length=3, radius=3.255, center=<2 0 0>, rotation=<rot 90 0 0>}]
 dyn_shirt_01_r [{shapeType=Capsule, length=3, radius=3.255, center=<2 0 0>, rotation=<rot 90 0 0>}]
 dyn_shirt_bck_01 [{shapeType=Capsule, length=2.75, radius=3.255, center=<1 0 0>, rotation=<rot 90 0 0>}]
 dyn_shirt_bck_01_l [{shapeType=Capsule, length=2.75, radius=3.255, center=<1 0 0>, rotation=<rot 90 0 0>}]
 dyn_shirt_bck_01_r [{shapeType=Capsule, length=2.75, radius=3.255, center=<1 0 0>, rotation=<rot 90 0 0>}]
 dyn_shirt_frt_01_l [{shapeType=Capsule, length=1, radius=1}]
 dyn_shirt_frt_01_r [{shapeType=Capsule, length=1, radius=1}]
 dyn_shirt_frt_02_l [{shapeType=Capsule, length=5.75, radius=2.75, center=<2.5 0 0>, rotation=<rot 90 0 0>}]
 dyn_shirt_frt_02_r [{shapeType=Capsule, length=5.75, radius=2.75, center=<2.5 0 0>, rotation=<rot 90 0 0>}]
}
constraints [
 {bone1Name=dyn_driver_shirt_bck, bone2Name=dyn_parent_shirt_bck, swing1LimitDegrees=25, swing2LimitDegrees=25, twistLimitDegrees=5}
 {bone1Name=dyn_driver_shirt_bck_r, bone2Name=dyn_parent_shirt_bck_r, swing1LimitDegrees=25, swing2LimitDegrees=25, twistLimitDegrees=5}
 {bone1Name=dyn_driver_shirt_bck_l, bone2Name=dyn_parent_shirt_bck_l, swing1LimitDegrees=25, swing2LimitDegrees=25, twistLimitDegrees=5}
 {bone1Name=dyn_shirt_bck_01, bone2Name=dyn_driver_shirt_bck, swing1LimitDegrees=25, swing2LimitDegrees=25, twistLimitDegrees=5}
 {bone1Name=dyn_shirt_bck_01_r, bone2Name=dyn_driver_shirt_bck_r, swing1LimitDegrees=25, swing2LimitDegrees=25, twistLimitDegrees=5}
 {
  bone1Name=dyn_shirt_bck_01_l
  bone2Name=dyn_driver_shirt_bck_l
  swing1LimitDegrees=45
  swing1Motion=Free
  swing2LimitDegrees=45
  swing2Motion=Free
  twistLimitDegrees=45
  twistMotion=Free
 }
 {bone1Name=dyn_shirt_01_l, bone2Name=dyn_driver_shirt_l, swing1LimitDegrees=28.4286, swing2LimitDegrees=33.5714, twistLimitDegrees=5}
 {bone1Name=dyn_driver_shirt_l, bone2Name=dyn_parent_shirt_l, swing1LimitDegrees=20, swing2LimitDegrees=20, twistLimitDegrees=5}
 {bone1Name=dyn_driver_shirt_r, bone2Name=dyn_parent_shirt_r, swing1LimitDegrees=20, swing2LimitDegrees=20, twistLimitDegrees=5}
 {bone1Name=dyn_shirt_01_r, bone2Name=dyn_driver_shirt_r, swing1LimitDegrees=28.4286, swing2LimitDegrees=33.5714, twistLimitDegrees=5}
 {bone1Name=dyn_frill_01_l, bone2Name=dyn_parent_frill_l, swing1LimitDegrees=32, swing2LimitDegrees=20, twistLimitDegrees=5}
 {bone1Name=dyn_shirt_frt_02_l, bone2Name=dyn_shirt_frt_01_l, swing1LimitDegrees=20, swing2LimitDegrees=25, twistLimitDegrees=5}
 {bone1Name=dyn_shirt_frt_02_r, bone2Name=dyn_shirt_frt_01_r, swing1LimitDegrees=20, swing2LimitDegrees=25, twistLimitDegrees=5}
]
