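# Post-copy AnimBlueprint for the F_MED_Rainbow_Hat character body, adding secondary motion to plushies, grenades, clothing and boot straps.
# It copies the attached parent's pose, then applies weighted constraints, flap nodes and pose-alpha-driven corrective offsets, cached as 'Pose'.
# An AnimDynamics chain adds jiggle to accessories, sleeves and socks.
# Output switches between RigidBody simulation, with a separate front-end physics asset, and a constraint-only fallback on the cached 'Dynamics' pose.
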AnimBlueprint /BRCosmetics/Characters/Player/Female/Medium/Bodies/F_MED_Rainbow_Hat/Meshes/F_MED_Rainbow_Hat_AnimBP
target_skeleton=&/F_MED_Rainbow_Hat_Skeleton
graphs [
 AnimGraph {nodes [
   "Constraint - dyn_parent_shirt_frt_l (PPPP)"!(
    ComponentPose
    Alpha=0.3
    ConstraintWeights_0=0.0266
    ConstraintWeights_1:float=0.4551
    ConstraintWeights_2:float=0.0419
    ConstraintWeights_3:float=0.4765
   ) -> (Pose) {
    alphaBoolBlend {blendOption=Linear}
    boneToModify {boneName=dyn_parent_shirt_frt_l}
    constraintSetup [
     {targetBone {boneName=spine_01}}
     {targetBone {boneName=thigh_twist_01_l}}
     {targetBone {boneName=thigh_twist_01_r}}
     {targetBone {boneName=pelvis}}
    ]
    constraintWeights [1,1,1,1]
   }
   "Transform (Modify) Bone - Bone: plushC_01"!(ComponentPose, Alpha=Get CrouchWalk() -> (CrouchWalk:double), Translation="-2.762,-0.024,0.549", Rotation="6.373,36.752,-15.751") -> (Pose) {
    rotationMode=BMM_Additive
    rotationSpace=BCS_BoneSpace
    translationMode=BMM_Additive
    translationSpace=BCS_BoneSpace
    alphaBoolBlend {blendOption=Linear}
    alphaScaleBiasClamp {bInterpResult=true, interpSpeedDecreasing=15, interpSpeedIncreasing=15}
    boneToModify {boneName=plushC_01}
    rotation=<rot 6.373 36.7519 -15.7511>
    translation=<-2.762 -0.024 0.5486>
   }
   n1 = RigidBody(ComponentPose=Local To Component(LocalPose=Use cached pose 'Pose'())) {
    bClampLinearTranslationLimitToRefPose=true
    bOverrideWorldGravity=true
    overridePhysicsAsset=&/F_MED_Rainbow_Hat_FEPhysics
    simulationSpace=BaseBoneSpace
    alphaBoolBlend {blendOption=Linear}
    baseBoneRef {boneName=root}
    componentAppliedLinearAccClamp=<1100 1100 2500>
    componentLinearAccScale=<0.5 0.5 0.5>
    componentLinearVelScale=<0.5 0.5 0.5>
    overrideWorldGravity {z=-980}
   }
   n2 = RigidBody(ComponentPose=@n1) {
    bClampLinearTranslationLimitToRefPose=true
    bOverrideWorldGravity=true
    overridePhysicsAsset=&/F_MED_Rainbow_Hat_Skirt_Physics
    simulationSpace=BaseBoneSpace
    alphaBoolBlend {blendOption=Linear}
    baseBoneRef {boneName=pelvis}
    componentAppliedLinearAccClamp=<1100 1100 1100>
    componentLinearAccScale=<0.5 0.5 0.5>
    componentLinearVelScale=<0.5 0.5 0.5>
    overrideWorldGravity {z=-980}
   }
   n3 = Component To Local(ComponentPose=@n2)
   n4 = RigidBody(ComponentPose=Local To Component(LocalPose=Use cached pose 'Dynamics'())) {
    bClampLinearTranslationLimitToRefPose=true
    bOverrideWorldGravity=true
    overridePhysicsAsset=&/F_MED_Rainbow_Hat_Physics
    simulationSpace=BaseBoneSpace
    alphaBoolBlend {blendOption=Linear}
    baseBoneRef {boneName=root}
    componentAppliedLinearAccClamp=<1100 1100 2500>
    componentLinearAccScale=<0.5 0.5 0.5>
    componentLinearVelScale=<0.5 0.5 0.5>
    overrideWorldGravity {z=-980}
   }
   n5 = RigidBody(ComponentPose=@n4) {
    bClampLinearTranslationLimitToRefPose=true
    bOverrideWorldGravity=true
    overridePhysicsAsset=&F_MED_Rainbow_Hat_Skirt_Physics
    simulationSpace=BaseBoneSpace
    alphaBoolBlend {blendOption=Linear}
    baseBoneRef {boneName=pelvis}
    componentAppliedLinearAccClamp=<1100 1100 1100>
    componentLinearAccScale=<0.5 0.5 0.5>
    componentLinearVelScale=<0.5 0.5 0.5>
    overrideWorldGravity {z=-980}
   }
   n6 = Component To Local(ComponentPose=@n5)
   n7 = Get bIsInFrontEnd()
   n8 = "Constraint - plushA_02 (PP)"!(ComponentPose=Local To Component(LocalPose=Use cached pose 'Dynamics'()), Alpha=1, ConstraintWeights_0=0.8, ConstraintWeights_1:float=0.2) -> (Pose) {
    alphaBoolBlend {blendOption=Linear}
    boneToModify {boneName=plushA_02}
    constraintSetup [{targetBone {boneName=thigh_l}}, {targetBone {boneName=pelvis}}]
    constraintWeights [1,1]
   }
   n9 = "Constraint - plushE_02 (PP)"!(ComponentPose=@n8, Alpha=0.2, ConstraintWeights_0=0.8, ConstraintWeights_1:float=0.2) -> (Pose) {
    alphaBoolBlend {blendOption=Linear}
    boneToModify {boneName=plushE_02}
    constraintSetup [{targetBone {boneName=thigh_l}}, {targetBone {boneName=pelvis}}]
    constraintWeights [1,1]
   }
   n10 = "Constraint - plushF_02 (PP)"!(ComponentPose=@n9, Alpha=0.2, ConstraintWeights_0=0.8, ConstraintWeights_1:float=0.2) -> (Pose) {
    alphaBoolBlend {blendOption=Linear}
    boneToModify {boneName=plushF_02}
    constraintSetup [{targetBone {boneName=thigh_l}}, {targetBone {boneName=pelvis}}]
    constraintWeights [1,1]
   }
   n11 = "Constraint - grenadeA_02 (PP)"!(ComponentPose=@n10, Alpha=1, ConstraintWeights_0=0.8, ConstraintWeights_1:float=0.2) -> (Pose) {
    alphaBoolBlend {blendOption=Linear}
    boneToModify {boneName=grenadeA_02}
    constraintSetup [{targetBone {boneName=thigh_l}}, {targetBone {boneName=pelvis}}]
    constraintWeights [1,1]
   }
   n12 = "Constraint - grenadeB_02 (PP)"!(ComponentPose=@n11, Alpha=1, ConstraintWeights_0=0.8, ConstraintWeights_1:float=0.2) -> (Pose) {
    alphaBoolBlend {blendOption=Linear}
    boneToModify {boneName=grenadeB_02}
    constraintSetup [{targetBone {boneName=thigh_l}}, {targetBone {boneName=pelvis}}]
    constraintWeights [1,1]
   }
   n13 = "Constraint - plushB_02 (PP)"!(ComponentPose=@n12, Alpha=1, ConstraintWeights_0=0.8, ConstraintWeights_1:float=0.2) -> (Pose) {
    alphaBoolBlend {blendOption=Linear}
    boneToModify {boneName=plushB_02}
    constraintSetup [{targetBone {boneName=thigh_r}}, {targetBone {boneName=pelvis}}]
    constraintWeights [1,1]
   }
   n14 = "Constraint - plushC_02 (PP)"!(ComponentPose=@n13, Alpha=1, ConstraintWeights_0=0.8, ConstraintWeights_1:float=0.2) -> (Pose) {
    alphaBoolBlend {blendOption=Linear}
    boneToModify {boneName=plushC_02}
    constraintSetup [{targetBone {boneName=thigh_r}}, {targetBone {boneName=pelvis}}]
    constraintWeights [1,1]
   }
   n15 = "Constraint - plushD_02 (PP)"!(ComponentPose=@n14, Alpha=1, ConstraintWeights_0=0.8, ConstraintWeights_1:float=0.2) -> (Pose) {
    alphaBoolBlend {blendOption=Linear}
    boneToModify {boneName=plushD_02}
    constraintSetup [{targetBone {boneName=thigh_r}}, {targetBone {boneName=pelvis}}]
    constraintWeights [1,1]
   }
   Output Pose(Result=Blend Poses by bool(
     BlendPose_0=Blend Poses by bool(BlendPose_0=@n3, BlendPose_1=@n6, bActiveValue=@n7) {blendType=Linear}
     BlendPose_1=Component To Local(ComponentPose=@n15)
     bActiveValue=Get bRigidBodyEnabled()
    ) {blendType=Linear}) {layerGroup=None}
   n16 = "Constraint - plushA_01 (PPP)"!(
    ComponentPose=Local To Component(LocalPose=Copy Pose From Mesh() {bUseAttachedParent=true})
    Alpha=1
    ConstraintWeights_0=0.945
    ConstraintWeights_1:float=0.0454
    ConstraintWeights_2:float=0.0097
   ) -> (Pose) {
    alphaBoolBlend {blendOption=Linear}
    boneToModify {boneName=plushA_01}
    constraintSetup [{targetBone {boneName=pelvis}}, {targetBone {boneName=thigh_twist_01_l}}, {targetBone {boneName=thigh_twist_01_r}}]
    constraintWeights [1,1,1]
   }
   n17 = "Constraint - plushB_01 (PPPP)"!(
    ComponentPose=@n16
    Alpha=1
    ConstraintWeights_0=0.0156
    ConstraintWeights_1:float=0.7342
    ConstraintWeights_2:float=0.0157
    ConstraintWeights_3:float=0.2345
   ) -> (Pose) {
    alphaBoolBlend {blendOption=Linear}
    boneToModify {boneName=plushB_01}
    constraintSetup [
     {targetBone {boneName=thigh_r}}
     {targetBone {boneName=pelvis}}
     {targetBone {boneName=thigh_twist_01_l}}
     {targetBone {boneName=thigh_twist_01_r}}
    ]
    constraintWeights [1,1,1,1]
   }
   n18 = "Constraint - plushC_01 (PPP)"!(ComponentPose=@n17, Alpha=1, ConstraintWeights_0=0.0264, ConstraintWeights_1:float=0.5961, ConstraintWeights_2:float=0.3775) -> (Pose) {
    alphaBoolBlend {blendOption=Linear}
    boneToModify {boneName=plushC_01}
    constraintSetup [{targetBone {boneName=thigh_r}}, {targetBone {boneName=pelvis}}, {targetBone {boneName=thigh_twist_01_r}}]
    constraintWeights [1,1,1]
   }
   n19 = "Constraint - plushD_01 (PP)"!(ComponentPose=@n18, Alpha=1, ConstraintWeights_0=0.9503, ConstraintWeights_1:float=0.0497) -> (Pose) {
    alphaBoolBlend {blendOption=Linear}
    boneToModify {boneName=plushD_01}
    constraintSetup [{targetBone {boneName=pelvis}}, {targetBone {boneName=thigh_twist_01_r}}]
    constraintWeights [1,1]
   }
   n20 = "Constraint - plushE_01 (PP)"!(ComponentPose=@n19, Alpha=0.2, ConstraintWeights_0=0.9848, ConstraintWeights_1:float=0.0152) -> (Pose) {
    alphaBoolBlend {blendOption=Linear}
    boneToModify {boneName=plushE_01}
    constraintSetup [{targetBone {boneName=pelvis}}, {targetBone {boneName=thigh_twist_01_l}}]
    constraintWeights [1,1]
   }
   n21 = "Constraint - plushF_01 (PP)"!(ComponentPose=@n20, Alpha=1, ConstraintWeights_0=0.9538, ConstraintWeights_1:float=0.0462) -> (Pose) {
    alphaBoolBlend {blendOption=Linear}
    boneToModify {boneName=plushF_01}
    constraintSetup [{targetBone {boneName=pelvis}}, {targetBone {boneName=thigh_twist_01_l}}]
    constraintWeights [1,1]
   }
   n22 = "Constraint - grenadeA_01 (P)"!(ComponentPose=@n21, Alpha=1, ConstraintWeights_0=1, ConstraintWeights_1:float=0) -> (Pose) {
    alphaBoolBlend {blendOption=Linear}
    boneToModify {boneName=grenadeA_01}
    constraintSetup [{targetBone {boneName=pelvis}}]
    constraintWeights [1,0]
   }
   n23 = "Constraint - grenadeB_01 (PP)"!(ComponentPose=@n22, Alpha=1, ConstraintWeights_0=0.9741, ConstraintWeights_1:float=0.0259) -> (Pose) {
    alphaBoolBlend {blendOption=Linear}
    boneToModify {boneName=grenadeB_01}
    constraintSetup [{targetBone {boneName=pelvis}}, {targetBone {boneName=thigh_twist_01_l}}]
    constraintWeights [1,1]
   }
   n24 = "Constraint - boot_strapA_01_l (PP)"!(ComponentPose=@n23, Alpha=1, ConstraintWeights_0=0.4257, ConstraintWeights_1:float=0.5743) -> (Pose) {
    alphaBoolBlend {blendOption=Linear}
    boneToModify {boneName=boot_strapA_01_l}
    constraintSetup [{targetBone {boneName=calf_l}}, {targetBone {boneName=calf_twist_02_l}}]
    constraintWeights [1,1]
   }
   n25 = "Constraint - boot_strapB_01_l (PP)"!(ComponentPose=@n24, Alpha=1, ConstraintWeights_0=0.7452, ConstraintWeights_1:float=0.2548) -> (Pose) {
    alphaBoolBlend {blendOption=Linear}
    boneToModify {boneName=boot_strapB_01_l}
    constraintSetup [{targetBone {boneName=calf_l}}, {targetBone {boneName=calf_twist_02_l}}]
    constraintWeights [1,1]
   }
   n26 = Anim Dynamics - Bone: boot_strapB_02_l(ComponentPose=@n25, Alpha=1, ExternalForce="0,0,0") {
    angularDampingOverride=1
    angularSpringConstant=1500
    bAngularSpring=true
    bGravityOverrideInSimSpace=true
    bOverrideAngularDamping=true
    bUseGravityOverride=true
    gravityScale=0.35
    simulationSpace=RootRelative
    alphaBoolBlend {blendOption=Linear}
    boundBone {boneName=boot_strapB_02_l}
    componentAppliedLinearAccClamp=<3000 3000 3500>
    componentLinearAccScale=<0.3 0.3 0.3>
    componentLinearVelScale=<0.3 0.3 1.5>
    gravityOverride {z=-980}
    physicsBodyDefinitions [{
      sphereCollisionRadius=0
      boundBone {boneName=boot_strapB_02_l}
      boxExtents=<8 8 11>
      constraintSetup {angularLimitsMax=<5 5 30>, angularLimitsMin=<-5 -25 -30>, angularTarget=<0 0 0>}
      localJointOffset=<0 2 0>
     }]
   }
   n27 = Anim Dynamics - Bone: boot_strapA_02_l(ComponentPose=@n26, Alpha=0.65, ExternalForce="0,0,0") {
    angularDampingOverride=1
    angularSpringConstant=1500
    bAngularSpring=true
    bGravityOverrideInSimSpace=true
    bOverrideAngularDamping=true
    bUseGravityOverride=true
    gravityScale=0.35
    simulationSpace=RootRelative
    alphaBoolBlend {blendOption=Linear}
    boundBone {boneName=boot_strapA_02_l}
    componentAppliedLinearAccClamp=<3000 3000 3500>
    componentLinearAccScale=<0.3 0.3 0.3>
    componentLinearVelScale=<0.3 0.3 1.5>
    gravityOverride {z=-925}
    physicsBodyDefinitions [{
      sphereCollisionRadius=0
      boundBone {boneName=boot_strapA_02_l}
      boxExtents=<19 22 16>
      constraintSetup {angularLimitsMax=<5 5 30>, angularLimitsMin=<-5 -25 -30>, angularTarget=<0 0 0>}
      localJointOffset=<0 4 0>
     }]
   }
   n28 = Get crouch_alpha() -> (crouch_alpha:double)
   n29 = "Transform (Modify) Bone - Bone: grenadeA_01"(ComponentPose=@n27, Alpha=@n28, Translation="-6.762,-6.710,-7.729", Rotation="-14.186,7.988,-19.357", Scale="1,1,1") {
    rotationMode=BMM_Additive
    rotationSpace=BCS_BoneSpace
    translationMode=BMM_Additive
    translationSpace=BCS_BoneSpace
    alphaBoolBlend {blendOption=Linear}
    alphaScaleBiasClamp {bInterpResult=true, interpSpeedDecreasing=15, interpSpeedIncreasing=15}
    boneToModify {boneName=grenadeA_01}
    rotation=<rot -14.186 7.988 -19.357>
    translation=<-6.762 -6.71 -7.729>
   }
   n30 = "Transform (Modify) Bone - Bone: grenadeB_01"(ComponentPose=@n29, Alpha=@n28, Translation="0.000,-1.781,0.042", Rotation="0.849,-10.945,4.390", Scale="1,1,1") {
    rotationMode=BMM_Additive
    rotationSpace=BCS_BoneSpace
    translationMode=BMM_Additive
    translationSpace=BCS_BoneSpace
    alphaBoolBlend {blendOption=Linear}
    alphaScaleBiasClamp {bInterpResult=true, interpSpeedDecreasing=15, interpSpeedIncreasing=15}
    boneToModify {boneName=grenadeB_01}
    rotation=<rot 0.849 -10.9454 4.39>
    translation {y=-1.781, z=0.042}
   }
   n31 = "Transform (Modify) Bone - Bone: plushB_01"!(ComponentPose=@n30, Alpha=@n28, Rotation="-25,24,0") -> (Pose) {
    rotationMode=BMM_Additive
    rotationSpace=BCS_BoneSpace
    translationMode=BMM_Additive
    translationSpace=BCS_BoneSpace
    alphaBoolBlend {blendOption=Linear}
    alphaScaleBiasClamp {bInterpResult=true, interpSpeedDecreasing=15, interpSpeedIncreasing=15}
    boneToModify {boneName=plushB_01}
    rotation {pitch=-25, yaw=24}
   }
   n32 = "Transform (Modify) Bone - Bone: plushB_01"(ComponentPose=@n31, Alpha=1, Rotation="0,0,-45") {
    alpha=0
    rotationMode=BMM_Additive
    rotationSpace=BCS_BoneSpace
    translationMode=BMM_Additive
    translationSpace=BCS_BoneSpace
    alphaBoolBlend {blendOption=Linear}
    alphaScaleBiasClamp {bInterpResult=true, interpSpeedDecreasing=15, interpSpeedIncreasing=15}
    boneToModify {boneName=plushB_01}
    rotation {roll=-45}
   }
   n33 = "Transform (Modify) Bone - Bone: dyn_parent_shirt_frt_r"!(ComponentPose=@n32, Alpha=@n28, Rotation="-9.5,0.000,0.000") -> (Pose) {
    rotationMode=BMM_Additive
    rotationSpace=BCS_BoneSpace
    translationMode=BMM_Additive
    translationSpace=BCS_BoneSpace
    alphaBoolBlend {blendOption=Linear}
    alphaScaleBiasClamp {bInterpResult=true, interpSpeedDecreasing=15, interpSpeedIncreasing=15}
    boneToModify {boneName=dyn_parent_shirt_frt_r}
    rotation {pitch=-9.5}
   }
   n34 = Flap Node - Bone: plushA_02!(ComponentPose:Component Space Pose Link Structure=@n33, Alpha:float=1) -> (Pose:Component Space Pose Link Structure) {
    is=title FortAnimGraphNode_Flap
    bUseClamp=true
    targetClampMax=10
    alphaBoolBlend {blendOption=Linear}
    sourceDrives [{
      bUseRange=false
      drivingCurve=null
      multiplier=1.1
      rangeMax=1
      rangeMin=-1
      remappedMax=1
      remappedMin=0
      sourceComponent=RotationZ
      useQuaternion=true
      sourceBone {boneName=thigh_l}
     }]
    targetBone {boneName=plushA_02}
   }
   n35 = Flap Node - Bone: plushB_02!(ComponentPose:Component Space Pose Link Structure=@n34, Alpha:float=1) -> (Pose:Component Space Pose Link Structure) {
    is=title FortAnimGraphNode_Flap
    bUseClamp=true
    targetClampMax=10
    alphaBoolBlend {blendOption=Linear}
    sourceDrives [
     {
      bUseRange=false
      drivingCurve=null
      multiplier=0.7
      rangeMax=1
      rangeMin=-1
      remappedMax=1
      remappedMin=0
      sourceComponent=RotationZ
      useQuaternion=true
      sourceBone {boneName=thigh_r}
     }
     {
      bUseRange=false
      drivingCurve=null
      multiplier=0.4
      rangeMax=1
      rangeMin=-1
      remappedMax=1
      remappedMin=0
      sourceComponent=RotationZ
      useQuaternion=true
      sourceBone {boneName=thigh_l}
     }
    ]
    targetBone {boneName=plushB_02}
   }
   n36 = Get thigh_fwd_alpha_r() -> (thigh_fwd_alpha_r:double)
   n37 = "Transform (Modify) Bone - Bone: plushA_01"(ComponentPose=@n35, Alpha=@n36, Translation="-0.199,-0.371,2.137", Rotation="10.432,0.906,-38.377", Scale="1,1,1") {
    rotationMode=BMM_Additive
    rotationSpace=BCS_BoneSpace
    translationMode=BMM_Additive
    translationSpace=BCS_BoneSpace
    alphaBoolBlend {blendOption=Linear}
    alphaScaleBiasClamp {bInterpResult=true, interpSpeedDecreasing=15, interpSpeedIncreasing=15}
    boneToModify {boneName=plushA_01}
    rotation=<rot 10.432 0.906 -38.3768>
    translation=<-0.199 -0.371 2.137>
   }
   n38 = "Transform (Modify) Bone - Bone: plushC_01"(ComponentPose=@n37, Alpha=@n36, Translation="-2.762,-0.024,0.549", Rotation="6.373,36.752,-15.751") {
    rotationMode=BMM_Additive
    rotationSpace=BCS_BoneSpace
    translationMode=BMM_Additive
    translationSpace=BCS_BoneSpace
    alphaBoolBlend {blendOption=Linear}
    alphaScaleBiasClamp {bInterpResult=true, interpSpeedDecreasing=15, interpSpeedIncreasing=15}
    boneToModify {boneName=plushC_01}
    rotation=<rot 6.373 36.7519 -15.7511>
    translation=<-2.762 -0.024 0.549>
   }
   n39 = "Constraint - dyn_thigh_shirt_l (P)"(ComponentPose=@n38, Alpha=1, ConstraintWeights_0=1) {
    alphaBoolBlend {blendOption=Linear}
    boneToModify {boneName=dyn_thigh_shirt_l}
    constraintSetup [{targetBone {boneName=thigh_twist_01_l}}]
    constraintWeights [1]
   }
   n40 = "Constraint - dyn_thigh_shirt_r (P)"(ComponentPose=@n39, Alpha=1, ConstraintWeights_0=1) {
    alphaBoolBlend {blendOption=Linear}
    boneToModify {boneName=dyn_thigh_shirt_r}
    constraintSetup [{targetBone {boneName=thigh_twist_01_r}}]
    constraintWeights [1]
   }
   n41 = "Constraint - dyn_parent_frill_l (PPP)"!(ComponentPose=@n40, Alpha=1, ConstraintWeights_0=0.4344, ConstraintWeights_1:float=0.3549, ConstraintWeights_2:float=0.2107) -> (Pose) {
    alphaBoolBlend {blendOption=Linear}
    boneToModify {boneName=dyn_parent_frill_l}
    constraintSetup [{targetBone {boneName=thigh_twist_01_l}}, {targetBone {boneName=dyn_frill_01_l}}, {targetBone {boneName=thigh_l}}]
    constraintWeights [1,1,1]
   }
   n42 = "Transform (Modify) Bone - Bone: dyn_driver_shirt_frt_l"(
    ComponentPose=@n41
    Alpha=Get thigh_fwd_alpha_l() -> (thigh_fwd_alpha_l:double)
    Translation="0.633,-0.000,1.75"
    Rotation="1,0.000,0.000"
    Scale="1,1,1"
   ) {
    rotationMode=BMM_Additive
    rotationSpace=BCS_BoneSpace
    translationMode=BMM_Additive
    translationSpace=BCS_BoneSpace
    alphaBoolBlend {blendOption=Linear}
    boneToModify {boneName=dyn_driver_shirt_frt_l}
    rotation {pitch=1}
    translation {x=0.633, z=1.75}
   }
   n43 = "Constraint - dyn_shirt_01_r (PPPP)"!(
    ComponentPose=@n42
    Alpha=1
    ConstraintWeights_0=0.1384
    ConstraintWeights_1:float=0.5911
    ConstraintWeights_2:float=0.2073
    ConstraintWeights_3:float=0.0632
   ) -> (Pose) {
    alphaBoolBlend {blendOption=Linear}
    boneToModify {boneName=dyn_shirt_01_r}
    constraintSetup [
     {targetBone {boneName=dyn_shirt_frt_02_r}}
     {targetBone {boneName=dyn_shirt_frt_01_r}}
     {targetBone {boneName=pelvis}}
     {targetBone {boneName=thigh_twist_01_r}}
    ]
    constraintWeights [1,1,1,1]
   }
   n44 = Get thigh_fwd_alpha_r()
   n45 = "Transform (Modify) Bone - Bone: dyn_shirt_frt_01_r"(ComponentPose=@n43, Alpha=@n44, Translation="-0.048,0.571,-0.030", Rotation="10,5.156,3.898", Scale="1,1,1") {
    rotationMode=BMM_Additive
    rotationSpace=BCS_BoneSpace
    translationMode=BMM_Additive
    translationSpace=BCS_BoneSpace
    alphaBoolBlend {blendOption=Linear}
    boneToModify {boneName=dyn_shirt_frt_01_r}
    rotation=<rot 10 5.156 3.898>
    translation=<-0.048 0.571 -0.03>
   }
   n46 = "Transform (Modify) Bone - Bone: dyn_parent_shirt_frt_r"!(ComponentPose=@n45, Alpha=Get CrouchWalk(), Translation:Vector="0,1,0", Rotation="7.092,-14.884,-22.225", Scale:Vector="1,1,1") -> (Pose) {
    rotationMode=BMM_Additive
    rotationSpace=BCS_BoneSpace
    translationMode=BMM_Additive
    translationSpace=BCS_BoneSpace
    alphaBoolBlend {blendOption=Linear}
    boneToModify {boneName=dyn_parent_shirt_frt_r}
    rotation=<rot 7.092 -14.8843 -22.2248>
    translation {y=1}
   }
   n47 = "Transform (Modify) Bone - Bone: dyn_parent_shirt_frt_l"(ComponentPose=@n46, Alpha=@n44, Translation="0,1,0", Rotation="0,0,-10", Scale="1,1,1") {
    rotationMode=BMM_Additive
    rotationSpace=BCS_BoneSpace
    translationMode=BMM_Additive
    translationSpace=BCS_BoneSpace
    alphaBoolBlend {blendOption=Linear}
    boneToModify {boneName=dyn_parent_shirt_frt_l}
    rotation {roll=-10}
    translation {y=1}
   }
   n48 = "Transform (Modify) Bone - Bone: dyn_driver_shirt_r"(ComponentPose=@n47, Alpha=1, Translation="0.056,0.366,2.969", Rotation="59.244,15.265,14.444", Scale="1,1,1") {
    rotationMode=BMM_Additive
    rotationSpace=BCS_BoneSpace
    translationMode=BMM_Additive
    translationSpace=BCS_BoneSpace
    alphaBoolBlend {blendOption=Linear}
    boneToModify {boneName=dyn_driver_shirt_r}
    rotation=<rot 59.244 15.2651 14.4441>
    translation=<0.056 0.366 2.969>
   }
   n49 = "Transform (Modify) Bone - Bone: dyn_shldr_frillD_01_r"(
    ComponentPose=@n48
    Alpha=Get arm_rot_high_r() -> (arm_rot_high_r:double)
    Translation="1.808,0.000,0.599"
    Rotation="18.087,-0.009,1.232"
    Scale="1,1,1"
   ) {
    rotationMode=BMM_Additive
    rotationSpace=BCS_BoneSpace
    translationMode=BMM_Additive
    translationSpace=BCS_BoneSpace
    alphaBoolBlend {blendOption=Linear}
    boneToModify {boneName=dyn_shldr_frillD_01_r}
    rotation=<rot 18.087 -0.009 1.232>
    translation {x=1.808, z=0.599}
   }
   Save cached pose 'Pose'(Pose=Component To Local(ComponentPose=@n49)) {cacheName=Pose}
   n50 = Anim Dynamics - Bone: plushA_02(ComponentPose=Local To Component(LocalPose=Use cached pose 'Pose'()), Alpha=0.2, ExternalForce="0,0,0") {
    angularSpringConstant=1500
    bGravityOverrideInSimSpace=true
    gravityScale=0.1
    linearSpringConstant=1000
    simulationSpace=BoneRelative
    alphaBoolBlend {blendOption=Linear}
    boundBone {boneName=plushA_02}
    componentLinearAccScale=<0.15 0.15 0.35>
    componentLinearVelScale=<0.15 15 0.35>
    gravityOverride {z=-980}
    physicsBodyDefinitions [{
      sphereCollisionRadius=0
      boundBone {boneName=plushA_02}
      boxExtents=<8 20 20>
      constraintSetup {angularLimitsMax=<2 5 30>, angularLimitsMin=<-5 -10 -110>, angularTarget=<0 0 0>}
      localJointOffset=<10 0 0>
     }]
   }
   n51 = Anim Dynamics - Bone: plushB_02(ComponentPose=@n50, Alpha=0.4, ExternalForce="0,0,0") {
    angularSpringConstant=1500
    bGravityOverrideInSimSpace=true
    gravityScale=0.1
    linearSpringConstant=1000
    simulationSpace=BoneRelative
    alphaBoolBlend {blendOption=Linear}
    boundBone {boneName=plushB_02}
    componentLinearAccScale=<0.15 0.15 0.2>
    componentLinearVelScale=<0.15 0.15 0.2>
    gravityOverride {z=-980}
    physicsBodyDefinitions [{
      sphereCollisionRadius=0
      boundBone {boneName=plushB_02}
      boxExtents=<8 20 20>
      constraintSetup {angularLimitsMax=<5 5 50>, angularLimitsMin=<-5 -10 -90>, angularTarget=<0 0 0>}
      localJointOffset=<2 0 0>
     }]
    relativeSpaceBone {boneName=pelvis}
   }
   n52 = Anim Dynamics - Bone: plushC_02(ComponentPose=@n51, Alpha=0.2, ExternalForce="0,0,0") {
    angularSpringConstant=3000
    bGravityOverrideInSimSpace=true
    gravityScale=0.15
    linearSpringConstant=1000
    alphaBoolBlend {blendOption=Linear}
    boundBone {boneName=plushC_02}
    componentAppliedLinearAccClamp=<6000 6000 6000>
    componentLinearAccScale=<0.15 0.2 0.05>
    componentLinearVelScale=<0.15 0.2 0.05>
    gravityOverride {z=-980}
    physicsBodyDefinitions [{
      sphereCollisionRadius=0
      boundBone {boneName=plushC_02}
      boxExtents=<2 10 10>
      constraintSetup {angularLimitsMax=<5 5 20>, angularLimitsMin=<-5 -2 -30>, angularTarget=<0 0 0>}
      localJointOffset=<0.5 0 0>
     }]
    relativeSpaceBone {boneName=root}
   }
   n53 = Anim Dynamics - Bone: plushC_ear_01_l(ComponentPose=@n52, Alpha=0.5, ExternalForce="0,0,0") {
    angularDampingOverride=0.9999
    angularSpringConstant=100
    bGravityOverrideInSimSpace=true
    bOverrideAngularDamping=true
    gravityScale=0.1
    simulationSpace=BoneRelative
    alphaBoolBlend {blendOption=Linear}
    boundBone {boneName=plushC_ear_01_l}
    componentLinearAccScale=<1.2 1.2 1.2>
    componentLinearVelScale=<1.2 1.2 1.2>
    physicsBodyDefinitions [{
      sphereCollisionRadius=0
      boundBone {boneName=plushC_ear_01_l}
      boxExtents=<20 10 10>
      constraintSetup {angularLimitsMax=<1 1 15>, angularLimitsMin=<-1 -30 -15>, angularTarget=<0 0 0>}
      localJointOffset=<2 0 0>
     }]
    relativeSpaceBone {boneName=pelvis}
   }
   n54 = Anim Dynamics - Bone: plushC_ear_01_r(ComponentPose=@n53, Alpha=0.5, ExternalForce="0,0,0") {
    angularDampingOverride=0.9999
    angularSpringConstant=100
    bGravityOverrideInSimSpace=true
    bOverrideAngularDamping=true
    gravityScale=0.1
    simulationSpace=BoneRelative
    alphaBoolBlend {blendOption=Linear}
    boundBone {boneName=plushC_ear_01_r}
    componentLinearAccScale=<1.2 1.2 1.2>
    componentLinearVelScale=<1.2 1.2 1.2>
    physicsBodyDefinitions [{
      sphereCollisionRadius=0
      boundBone {boneName=plushC_ear_01_r}
      boxExtents=<20 10 10>
      constraintSetup {angularLimitsMax=<1 1 15>, angularLimitsMin=<-1 -30 -15>, angularTarget=<0 0 0>}
      localJointOffset=<2 0 0>
     }]
    relativeSpaceBone {boneName=pelvis}
    sphericalLimits [
     {limitRadius=20, drivingBone {boneName=spine_04}, sphereLocalOffset=<-5 1 -12>}
     {limitRadius=20, drivingBone {boneName=spine_03}, sphereLocalOffset=<0 0 0>}
    ]
   }
   n55 = Anim Dynamics - Bone: plushE_02(ComponentPose=@n54, Alpha=0.2, ExternalForce="0,0,0") {
    angularSpringConstant=1500
    bGravityOverrideInSimSpace=true
    gravityScale=0.1
    linearSpringConstant=1000
    simulationSpace=BoneRelative
    alphaBoolBlend {blendOption=Linear}
    boundBone {boneName=plushE_02}
    componentLinearAccScale=<0.15 0.15 0.35>
    componentLinearVelScale=<0.15 0.15 0.35>
    gravityOverride {z=-980}
    physicsBodyDefinitions [{
      sphereCollisionRadius=0
      boundBone {boneName=plushE_02}
      boxExtents=<8 20 20>
      constraintSetup {angularLimitsMax=<5 5 20>, angularLimitsMin=<-5 -30 -120>, angularTarget=<0 0 0>}
      localJointOffset=<10 0 0>
     }]
   }
   n56 = Anim Dynamics - Bone: grenadeA_02(ComponentPose=@n55, Alpha=0.5, ExternalForce="0,0,0") {
    angularDampingOverride=1
    angularSpringConstant=1500
    bGravityOverrideInSimSpace=true
    bOverrideAngularDamping=true
    gravityScale=0.15
    simulationSpace=BoneRelative
    alphaBoolBlend {blendOption=Linear}
    boundBone {boneName=grenadeA_02}
    physicsBodyDefinitions [{
      sphereCollisionRadius=0
      boundBone {boneName=grenadeA_02}
      boxExtents=<10 20 20>
      constraintSetup {angularLimitsMax=<0 0 2>, angularLimitsMin=<0 0 -15>, angularTarget=<0 0 0>}
      localJointOffset=<5 0 0>
     }]
    relativeSpaceBone {boneName=pelvis}
   }
   n57 = Anim Dynamics - Bone: grenadeB_02(ComponentPose=@n56, Alpha=0.5, ExternalForce="0,0,0") {
    angularDampingOverride=1
    angularSpringConstant=1500
    bGravityOverrideInSimSpace=true
    bOverrideAngularDamping=true
    gravityScale=0.15
    simulationSpace=BoneRelative
    alphaBoolBlend {blendOption=Linear}
    boundBone {boneName=grenadeB_02}
    physicsBodyDefinitions [{
      sphereCollisionRadius=0
      boundBone {boneName=grenadeB_02}
      boxExtents=<10 20 20>
      constraintSetup {angularLimitsMax=<0 0 0>, angularLimitsMin=<0 0 -10>, angularTarget=<0 0 0>}
      localJointOffset=<8 0 0>
     }]
    relativeSpaceBone {boneName=pelvis}
   }
   n58 = Anim Dynamics - Bone: dyn_sleeve_01_l(ComponentPose=@n57, Alpha=0.3, ExternalForce="0,0,0") {
    bGravityOverrideInSimSpace=true
    bUseGravityOverride=true
    gravityScale=0.3
    lODThreshold=1
    simulationSpace=BoneRelative
    alphaBoolBlend {blendOption=Linear}
    alphaScaleBiasClamp {clampMax=0.09, clampMin=0.05}
    boundBone {boneName=dyn_sleeve_01_l}
    gravityOverride {z=-980}
    physicsBodyDefinitions [{
      sphereCollisionRadius=0
      boundBone {boneName=dyn_sleeve_01_l}
      boxExtents=<2 5 5>
      constraintSetup {angularLimitsMax=<0 0 0>, angularLimitsMin=<0 0 0>, angularTarget=<0 0 0>, linearAxesMax=<0.5 0.5 0.5>, linearAxesMin=<-1.5 -1.5 -1.5>}
      localJointOffset=<0 0 -2>
     }]
    relativeSpaceBone {boneName=pelvis}
   }
   n59 = Anim Dynamics - Bone: dyn_sleeve_02_l(ComponentPose=@n58, Alpha=0.3, ExternalForce="0,0,0") {
    bGravityOverrideInSimSpace=true
    bUseGravityOverride=true
    gravityScale=0.3
    lODThreshold=1
    simulationSpace=BoneRelative
    alphaBoolBlend {blendOption=Linear}
    alphaScaleBiasClamp {clampMax=0.09, clampMin=0.05}
    boundBone {boneName=dyn_sleeve_02_l}
    gravityOverride {z=-980}
    physicsBodyDefinitions [{
      sphereCollisionRadius=0
      boundBone {boneName=dyn_sleeve_02_l}
      boxExtents=<2 5 5>
      constraintSetup {angularLimitsMax=<0 0 0>, angularLimitsMin=<0 0 0>, angularTarget=<0 0 0>, linearAxesMax=<0.5 0.5 0.5>, linearAxesMin=<-1.5 -1.5 -1.5>}
      localJointOffset=<0 0 -2>
     }]
    relativeSpaceBone {boneName=pelvis}
   }
   n60 = Anim Dynamics - Bone: dyn_sleeve_03_l(ComponentPose=@n59, Alpha=0.3, ExternalForce="0,0,0") {
    bGravityOverrideInSimSpace=true
    bUseGravityOverride=true
    gravityScale=0.3
    lODThreshold=1
    simulationSpace=BoneRelative
    alphaBoolBlend {blendOption=Linear}
    alphaScaleBiasClamp {clampMax=0.09, clampMin=0.05}
    boundBone {boneName=dyn_sleeve_03_l}
    gravityOverride {z=-980}
    physicsBodyDefinitions [{
      sphereCollisionRadius=0
      boundBone {boneName=dyn_sleeve_03_l}
      boxExtents=<2 5 5>
      constraintSetup {angularLimitsMax=<0 0 0>, angularLimitsMin=<0 0 0>, angularTarget=<0 0 0>, linearAxesMax=<0.5 0.5 0.5>, linearAxesMin=<-1.5 -1.5 -1.5>}
      localJointOffset=<0 0 -2>
     }]
    relativeSpaceBone {boneName=pelvis}
   }
   n61 = Anim Dynamics - Bone: dyn_sleeve_04_l(ComponentPose=@n60, Alpha=0.3, ExternalForce="0,0,0") {
    bGravityOverrideInSimSpace=true
    bUseGravityOverride=true
    gravityScale=0.3
    lODThreshold=1
    simulationSpace=BoneRelative
    alphaBoolBlend {blendOption=Linear}
    alphaScaleBiasClamp {clampMax=0.09, clampMin=0.05}
    boundBone {boneName=dyn_sleeve_04_l}
    gravityOverride {z=-980}
    physicsBodyDefinitions [{
      sphereCollisionRadius=0
      boundBone {boneName=dyn_sleeve_04_l}
      boxExtents=<2 5 5>
      constraintSetup {angularLimitsMax=<0 0 0>, angularLimitsMin=<0 0 0>, angularTarget=<0 0 0>, linearAxesMax=<0.5 0.5 0.5>, linearAxesMin=<-1.5 -1.5 -1.5>}
      localJointOffset=<0 0 -2>
     }]
    relativeSpaceBone {boneName=pelvis}
   }
   n62 = Anim Dynamics - Bone: dyn_sleeve_01_r(ComponentPose=@n61, Alpha=0.3, ExternalForce="0,0,0") {
    bGravityOverrideInSimSpace=true
    bUseGravityOverride=true
    gravityScale=0.3
    lODThreshold=1
    simulationSpace=BoneRelative
    alphaBoolBlend {blendOption=Linear}
    alphaScaleBiasClamp {clampMax=0.09, clampMin=0.05}
    boundBone {boneName=dyn_sleeve_01_r}
    gravityOverride {z=-980}
    physicsBodyDefinitions [{
      sphereCollisionRadius=0
      boundBone {boneName=dyn_sleeve_01_r}
      boxExtents=<2 5 5>
      constraintSetup {angularLimitsMax=<0 0 0>, angularLimitsMin=<0 0 0>, angularTarget=<0 0 0>, linearAxesMax=<0.5 0.5 0.5>, linearAxesMin=<-1.5 -1.5 -1.5>}
      localJointOffset=<0 0 -2>
     }]
    relativeSpaceBone {boneName=pelvis}
   }
   n63 = Anim Dynamics - Bone: dyn_sleeve_02_r(ComponentPose=@n62, Alpha=0.3, ExternalForce="0,0,0") {
    bGravityOverrideInSimSpace=true
    bUseGravityOverride=true
    gravityScale=0.3
    lODThreshold=1
    simulationSpace=BoneRelative
    alphaBoolBlend {blendOption=Linear}
    alphaScaleBiasClamp {clampMax=0.09, clampMin=0.05}
    boundBone {boneName=dyn_sleeve_02_r}
    gravityOverride {z=-980}
    physicsBodyDefinitions [{
      sphereCollisionRadius=0
      boundBone {boneName=dyn_sleeve_02_r}
      boxExtents=<2 5 5>
      constraintSetup {angularLimitsMax=<0 0 0>, angularLimitsMin=<0 0 0>, angularTarget=<0 0 0>, linearAxesMax=<0.5 0.5 0.5>, linearAxesMin=<-1.5 -1.5 -1.5>}
      localJointOffset=<0 0 -2>
     }]
    relativeSpaceBone {boneName=pelvis}
   }
   n64 = Anim Dynamics - Bone: dyn_sleeve_03_r(ComponentPose=@n63, Alpha=0.3, ExternalForce="0,0,0") {
    bGravityOverrideInSimSpace=true
    bUseGravityOverride=true
    gravityScale=0.3
    lODThreshold=1
    simulationSpace=BoneRelative
    alphaBoolBlend {blendOption=Linear}
    alphaScaleBiasClamp {clampMax=0.09, clampMin=0.05}
    boundBone {boneName=dyn_sleeve_03_r}
    gravityOverride {z=-980}
    physicsBodyDefinitions [{
      sphereCollisionRadius=0
      boundBone {boneName=dyn_sleeve_03_r}
      boxExtents=<2 5 5>
      constraintSetup {angularLimitsMax=<0 0 0>, angularLimitsMin=<0 0 0>, angularTarget=<0 0 0>, linearAxesMax=<0.5 0.5 0.5>, linearAxesMin=<-1.5 -1.5 -1.5>}
      localJointOffset=<0 0 -2>
     }]
    relativeSpaceBone {boneName=pelvis}
   }
   n65 = Anim Dynamics - Bone: dyn_sleeve_04_r(ComponentPose=@n64, Alpha=0.3, ExternalForce="0,0,0") {
    bGravityOverrideInSimSpace=true
    bUseGravityOverride=true
    gravityScale=0.3
    lODThreshold=1
    simulationSpace=BoneRelative
    alphaBoolBlend {blendOption=Linear}
    alphaScaleBiasClamp {clampMax=0.09, clampMin=0.05}
    boundBone {boneName=dyn_sleeve_04_r}
    gravityOverride {z=-980}
    physicsBodyDefinitions [{
      sphereCollisionRadius=0
      boundBone {boneName=dyn_sleeve_04_r}
      boxExtents=<2 5 5>
      constraintSetup {angularLimitsMax=<0 0 0>, angularLimitsMin=<0 0 0>, angularTarget=<0 0 0>, linearAxesMax=<0.5 0.5 0.5>, linearAxesMin=<-1.5 -1.5 -1.5>}
      localJointOffset=<0 0 -2>
     }]
    relativeSpaceBone {boneName=pelvis}
   }
   n66 = Anim Dynamics - Bone: dyn_socksA_l(ComponentPose=@n65, Alpha=0.55, ExternalForce="0,0,0") {
    bGravityOverrideInSimSpace=true
    gravityScale=0.1
    lODThreshold=1
    simulationSpace=BoneRelative
    alphaBoolBlend {blendOption=Linear}
    alphaScaleBiasClamp {clampMax=0.09, clampMin=0.05}
    boundBone {boneName=dyn_socksA_l}
    physicsBodyDefinitions [{
      sphereCollisionRadius=0
      boundBone {boneName=dyn_socksA_l}
      boxExtents=<2 5 5>
      constraintSetup {angularLimitsMax=<0 0 0>, angularLimitsMin=<0 0 0>, angularTarget=<0 0 0>, linearAxesMax=<0.8 1 0.5>, linearAxesMin=<-0.5 -2 -1>}
      localJointOffset=<0 0 -1>
     }]
    relativeSpaceBone {boneName=thigh_l}
   }
   n67 = Anim Dynamics - Bone: dyn_socksB_l(ComponentPose=@n66, Alpha=0.55, ExternalForce="0,0,0") {
    bGravityOverrideInSimSpace=true
    gravityScale=0.1
    lODThreshold=1
    simulationSpace=BoneRelative
    alphaBoolBlend {blendOption=Linear}
    alphaScaleBiasClamp {clampMax=0.09, clampMin=0.05}
    boundBone {boneName=dyn_socksB_l}
    physicsBodyDefinitions [{
      sphereCollisionRadius=0
      boundBone {boneName=dyn_socksB_l}
      boxExtents=<2 5 5>
      constraintSetup {angularLimitsMax=<0 0 0>, angularLimitsMin=<0 0 0>, angularTarget=<0 0 0>, linearAxesMax=<0.8 1 0.5>, linearAxesMin=<-0.5 -2 -1>}
      localJointOffset=<0 0 -1>
     }]
    relativeSpaceBone {boneName=thigh_l}
   }
   n68 = Anim Dynamics - Bone: dyn_socksC_l(ComponentPose=@n67, Alpha=0.55, ExternalForce="0,0,0") {
    bGravityOverrideInSimSpace=true
    gravityScale=0.1
    lODThreshold=1
    simulationSpace=BoneRelative
    alphaBoolBlend {blendOption=Linear}
    alphaScaleBiasClamp {clampMax=0.09, clampMin=0.05}
    boundBone {boneName=dyn_socksC_l}
    physicsBodyDefinitions [{
      sphereCollisionRadius=0
      boundBone {boneName=dyn_socksC_l}
      boxExtents=<2 5 5>
      constraintSetup {angularLimitsMax=<0 0 0>, angularLimitsMin=<0 0 0>, angularTarget=<0 0 0>, linearAxesMax=<0.8 1 0.5>, linearAxesMin=<-0.5 -2 -1>}
      localJointOffset=<0 0 -1>
     }]
    relativeSpaceBone {boneName=thigh_l}
   }
   n69 = Anim Dynamics - Bone: dyn_socksD_l(ComponentPose=@n68, Alpha=0.55, ExternalForce="0,0,0") {
    bGravityOverrideInSimSpace=true
    gravityScale=0.1
    lODThreshold=1
    simulationSpace=BoneRelative
    alphaBoolBlend {blendOption=Linear}
    alphaScaleBiasClamp {clampMax=0.09, clampMin=0.05}
    boundBone {boneName=dyn_socksD_l}
    physicsBodyDefinitions [{
      sphereCollisionRadius=0
      boundBone {boneName=dyn_socksD_l}
      boxExtents=<2 5 5>
      constraintSetup {angularLimitsMax=<0 0 0>, angularLimitsMin=<0 0 0>, angularTarget=<0 0 0>, linearAxesMax=<0.8 1 0.5>, linearAxesMin=<-0.5 -2 -1>}
      localJointOffset=<0 0 -1>
     }]
    relativeSpaceBone {boneName=thigh_l}
   }
   n70 = Anim Dynamics - Bone: dyn_socksE_l(ComponentPose=@n69, Alpha=0.55, ExternalForce="0,0,0") {
    bGravityOverrideInSimSpace=true
    gravityScale=0.1
    lODThreshold=1
    simulationSpace=BoneRelative
    alphaBoolBlend {blendOption=Linear}
    alphaScaleBiasClamp {clampMax=0.09, clampMin=0.05}
    boundBone {boneName=dyn_socksE_l}
    physicsBodyDefinitions [{
      sphereCollisionRadius=0
      boundBone {boneName=dyn_socksE_l}
      boxExtents=<2 5 5>
      constraintSetup {angularLimitsMax=<0 0 0>, angularLimitsMin=<0 0 0>, angularTarget=<0 0 0>, linearAxesMax=<0.8 1 0.5>, linearAxesMin=<-0.5 -2 -1>}
      localJointOffset=<0 0 -1>
     }]
    relativeSpaceBone {boneName=thigh_l}
   }
   n71 = Anim Dynamics - Bone: dyn_socksA_r(ComponentPose=@n70, Alpha=0.55, ExternalForce="0,0,0") {
    bGravityOverrideInSimSpace=true
    gravityScale=0.1
    lODThreshold=1
    simulationSpace=BoneRelative
    alphaBoolBlend {blendOption=Linear}
    alphaScaleBiasClamp {clampMax=0.09, clampMin=0.05}
    boundBone {boneName=dyn_socksA_r}
    physicsBodyDefinitions [{
      sphereCollisionRadius=0
      boundBone {boneName=dyn_socksA_r}
      boxExtents=<2 5 5>
      constraintSetup {angularLimitsMax=<0 0 0>, angularLimitsMin=<0 0 0>, angularTarget=<0 0 0>, linearAxesMax=<1 1 1.5>, linearAxesMin=<-0.5 -2 -0.5>}
      localJointOffset=<0 0 -1>
     }]
    relativeSpaceBone {boneName=thigh_r}
   }
   n72 = Anim Dynamics - Bone: dyn_socksB_r(ComponentPose=@n71, Alpha=0.55, ExternalForce="0,0,0") {
    bGravityOverrideInSimSpace=true
    gravityScale=0.1
    lODThreshold=1
    simulationSpace=BoneRelative
    alphaBoolBlend {blendOption=Linear}
    alphaScaleBiasClamp {clampMax=0.09, clampMin=0.05}
    boundBone {boneName=dyn_socksB_r}
    physicsBodyDefinitions [{
      sphereCollisionRadius=0
      boundBone {boneName=dyn_socksB_r}
      boxExtents=<2 5 5>
      constraintSetup {angularLimitsMax=<0 0 0>, angularLimitsMin=<0 0 0>, angularTarget=<0 0 0>, linearAxesMax=<1 1 1>, linearAxesMin=<-0.5 -2 -1>}
      localJointOffset=<0 0 -1>
     }]
    relativeSpaceBone {boneName=thigh_r}
   }
   n73 = Anim Dynamics - Bone: dyn_socksC_r(ComponentPose=@n72, Alpha=0.55, ExternalForce="0,0,0") {
    bGravityOverrideInSimSpace=true
    gravityScale=0.1
    lODThreshold=1
    simulationSpace=BoneRelative
    alphaBoolBlend {blendOption=Linear}
    alphaScaleBiasClamp {clampMax=0.09, clampMin=0.05}
    boundBone {boneName=dyn_socksC_r}
    physicsBodyDefinitions [{
      sphereCollisionRadius=0
      boundBone {boneName=dyn_socksC_r}
      boxExtents=<2 5 5>
      constraintSetup {angularLimitsMax=<0 0 0>, angularLimitsMin=<0 0 0>, angularTarget=<0 0 0>, linearAxesMax=<1 1 1.5>, linearAxesMin=<-0.5 -1 -0.5>}
      localJointOffset=<0 0 -1>
     }]
    relativeSpaceBone {boneName=thigh_r}
   }
   n74 = Anim Dynamics - Bone: dyn_socksD_r(ComponentPose=@n73, Alpha=0.55, ExternalForce="0,0,0") {
    bGravityOverrideInSimSpace=true
    gravityScale=0.1
    lODThreshold=1
    simulationSpace=BoneRelative
    alphaBoolBlend {blendOption=Linear}
    alphaScaleBiasClamp {clampMax=0.09, clampMin=0.05}
    boundBone {boneName=dyn_socksD_r}
    physicsBodyDefinitions [{
      sphereCollisionRadius=0
      boundBone {boneName=dyn_socksD_r}
      boxExtents=<2 5 5>
      constraintSetup {angularLimitsMax=<0 0 0>, angularLimitsMin=<0 0 0>, angularTarget=<0 0 0>, linearAxesMax=<1 1 1>, linearAxesMin=<-0.5 -2 -1>}
      localJointOffset=<0 0 -1>
     }]
    relativeSpaceBone {boneName=thigh_r}
   }
   n75 = Anim Dynamics - Bone: dyn_socksE_r(ComponentPose=@n74, Alpha=0.55, ExternalForce="0,0,0") {
    bGravityOverrideInSimSpace=true
    gravityScale=0.1
    lODThreshold=1
    simulationSpace=BoneRelative
    alphaBoolBlend {blendOption=Linear}
    alphaScaleBiasClamp {clampMax=0.09, clampMin=0.05}
    boundBone {boneName=dyn_socksE_r}
    physicsBodyDefinitions [{
      sphereCollisionRadius=0
      boundBone {boneName=dyn_socksE_r}
      boxExtents=<2 5 5>
      constraintSetup {angularLimitsMax=<0 0 0>, angularLimitsMin=<0 0 0>, angularTarget=<0 0 0>, linearAxesMax=<1 1 1>, linearAxesMin=<-0.5 -2 -1>}
      localJointOffset=<0 0 -1>
     }]
    relativeSpaceBone {boneName=thigh_r}
   }
   n76 = Anim Dynamics - Bone: dyn_socksF_r(ComponentPose=@n75, Alpha=0.55, ExternalForce="0,0,0") {
    bGravityOverrideInSimSpace=true
    gravityScale=0.1
    lODThreshold=1
    simulationSpace=BoneRelative
    alphaBoolBlend {blendOption=Linear}
    alphaScaleBiasClamp {clampMax=0.09, clampMin=0.05}
    boundBone {boneName=dyn_socksF_r}
    physicsBodyDefinitions [{
      sphereCollisionRadius=0
      boundBone {boneName=dyn_socksF_r}
      boxExtents=<2 5 5>
      constraintSetup {angularLimitsMax=<0 0 0>, angularLimitsMin=<0 0 0>, angularTarget=<0 0 0>, linearAxesMax=<1 1 1>, linearAxesMin=<-0.5 -2 -1>}
      localJointOffset=<0 0 -1>
     }]
    relativeSpaceBone {boneName=thigh_r}
   }
   n77 = Anim Dynamics - Bone: dyn_socksG_r(ComponentPose=@n76, Alpha=0.55, ExternalForce="0,0,0") {
    bGravityOverrideInSimSpace=true
    gravityScale=0.1
    lODThreshold=1
    simulationSpace=BoneRelative
    alphaBoolBlend {blendOption=Linear}
    alphaScaleBiasClamp {clampMax=0.09, clampMin=0.05}
    boundBone {boneName=dyn_socksG_r}
    physicsBodyDefinitions [{
      sphereCollisionRadius=0
      boundBone {boneName=dyn_socksG_r}
      boxExtents=<2 5 5>
      constraintSetup {angularLimitsMax=<0 0 0>, angularLimitsMin=<0 0 0>, angularTarget=<0 0 0>, linearAxesMax=<1 1 1>, linearAxesMin=<-0.5 -2 -1>}
      localJointOffset=<0 0 -1>
     }]
    relativeSpaceBone {boneName=thigh_r}
   }
   n78 = Anim Dynamics - Bone: dyn_socksH_r(ComponentPose=@n77, Alpha=0.55, ExternalForce="0,0,0") {
    bGravityOverrideInSimSpace=true
    gravityScale=0.1
    lODThreshold=1
    simulationSpace=BoneRelative
    alphaBoolBlend {blendOption=Linear}
    alphaScaleBiasClamp {clampMax=0.09, clampMin=0.05}
    boundBone {boneName=dyn_socksH_r}
    physicsBodyDefinitions [{
      sphereCollisionRadius=0
      boundBone {boneName=dyn_socksH_r}
      boxExtents=<2 5 5>
      constraintSetup {angularLimitsMax=<0 0 0>, angularLimitsMin=<0 0 0>, angularTarget=<0 0 0>, linearAxesMax=<1 1 1>, linearAxesMin=<-0.5 -2 -1>}
      localJointOffset=<0 0 -1>
     }]
    relativeSpaceBone {boneName=thigh_r}
   }
   Save cached pose 'Dynamics'(Pose=Component To Local(ComponentPose=@n78)) {cacheName=Dynamics}
  ]}
 EventGraph {nodes [
   Event BlueprintUpdateAnimation()
   > n1 = Sequence!(execute) -> (then_0, then_1, then_2:Exec, then_3:Exec, then_4:Exec)
   n2 = Branch(execute=@n1.then_4, Condition=AND Boolean(A=Get bIsMoving2D(), B=Get bIsCrouching()))
   > Set CrouchWalk(CrouchWalk:double=1) -> (Output_Get:double)
   Set CrouchWalk(execute=@n2.else, CrouchWalk=0)
   Set thigh_fwd_alpha_r(
    execute=@n1.then_0
    thigh_fwd_alpha_r:double=GetAlphaFromReferenceTransformDelta(
     SkeletalMeshComponent=Get Skel_Mesh()
     BoneName=thigh_r
     BaseName=pelvis
     DeltaComponentType=QuaternionTwist
     InputRange=(X=0,Y=100)
     TwistAxis="0,0,1"
    )
   ) -> (Output_Get:double)
   n3 = Get Skel_Mesh()
   Set thigh_back_alpha_l(
    execute=@n1.then_2
    thigh_back_alpha_l:double=GetAlphaFromReferenceTransformDelta(SkeletalMeshComponent=@n3, BoneName=thigh_l, BaseName=pelvis, DeltaComponentType=QuaternionTwist, InputRange=(X=0,Y=-60), TwistAxis="0,0,1")
   ) -> (Output_Get:double)
   Set thigh_fwd_alpha_l(
    execute=@n1.then_1
    thigh_fwd_alpha_l:double=GetAlphaFromReferenceTransformDelta(SkeletalMeshComponent=@n3, BoneName=thigh_l, BaseName=pelvis, DeltaComponentType=QuaternionTwist, InputRange=(X=0,Y=108), TwistAxis="0,0,1")
   ) -> (Output_Get:double)
   Set arm_rot_high_r(
    execute=@n1.then_3
    arm_rot_high_r:double=GetAlphaFromReferenceTransformDelta(SkeletalMeshComponent=Get Skel_Mesh(), BoneName=upperarm_r, BaseName=clavicle_r, InputRange=(X=30,Y=90), TwistAxis="0,0,1")
   ) -> (Output_Get:double)
   Event BlueprintInitializeAnimation()
   > Set Skel_Mesh(Skel_Mesh=GetOwnerSkeletalMeshComponent())
   > n4 = Get!(execute:Exec) -> (then:Exec, else:Exec, Skel_Mesh:Skeletal Mesh Component Object Reference)
   > Set Anim_Inst(Anim_Inst:Anim Instance Object Reference=GetAnimInstance(self=@n4)) -> (Output_Get:Anim Instance Object Reference)
   Event OnEndCrouching()
   > Set crouch_alpha(crouch_alpha:double=0) -> (Output_Get:double)
   Event OnBeginCrouching()
   > Set crouch_alpha(crouch_alpha=1)
  ]}
]
variables [
 Anim_Inst {type=AnimInstance, default=null}
 arm_rot_high_r {type=number}
 crouch_alpha {type=number}
 CrouchWalk {type=number}
 moving_alpha {type=number}
 Skel_Mesh {type=SkeletalMeshComponent}
 thigh_back_alpha_l {type=number}
 thigh_fwd_alpha_l {type=number}
 thigh_fwd_alpha_r {type=number}
]
