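# Physics asset for the F_MED_Rainbow_Hat face accessory mesh, giving capsule bodies and constraints to its hair and hood bones.
# Long hair chains hang from tiny head-linked driver bodies, with swing limits mostly widening tipward, except the free-moving final right segment.
# Short front hair and hood front/back constrain to the head, hood mouth to the neck.
# Head, neck, upper spine, clavicle and upper arm shapes form the body volume, joined by free constraints.
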PhysicsAsset /BRCosmetics/Characters/Player/Female/Medium/Bodies/F_MED_Rainbow_Hat/Meshes/Parts/F_MED_Rainbow_Hat_FE_FaceAcc_Physics
bodies {
 clavicle_l [{shapeType=Capsule, length=18.5, radius=8.5, center=<6.7926 -0.8364 4.9812>, rotation=<rot 90 0 0>}]
 clavicle_r [{shapeType=Capsule, length=18.5, radius=8.5, center=<-8.1631 5.6703 -3.8237>, rotation=<rot 90 -90 -95.4914>}]
 dyn_driver_lng_hair_l [{shapeType=Capsule, length=0.505, radius=0.505, rotation=<rot 0 -90 0>}]
 dyn_driver_rng_hair_r [{shapeType=Capsule, length=0.505, radius=0.505, rotation=<rot 0 -90 0>}]
 dyn_frnt_hair_01_l [{shapeType=Capsule, length=1.5, radius=2.5, center=<1 0 0>, rotation=<rot 90 0 0>}]
 dyn_frnt_hair_01_r [{shapeType=Capsule, length=1.5, radius=2.5, center=<1 0 0>, rotation=<rot 90 0 0>}]
 dyn_frnt_hair_02_l [{shapeType=Capsule, length=3, radius=2, center=<1.5535 0 0>, rotation=<rot 90 0 0>}]
 dyn_frnt_hair_02_r [{shapeType=Capsule, length=2.25, radius=3, center=<1 0 0>, rotation=<rot 90 0 0>}]
 dyn_hood_bck [{shapeType=Capsule, length=5.25, radius=3.25, center=<0.4 0 0>, rotation=<rot 90 0 0>}]
 dyn_hood_frt [{shapeType=Capsule, length=4.25, radius=3.75, center=<-0.4543 0 0>, rotation=<rot 90 0 0>}]
 dyn_hood_mouth [{shapeType=Capsule, length=4.5, radius=3, center=<0.4323 0 0>, rotation=<rot 90 0 0>}]
 dyn_lng_hair_01_l [{shapeType=Capsule, length=4.25, radius=2.75, rotation=<rot 90 0 0>}]
 dyn_lng_hair_02_l [{shapeType=Capsule, length=4.3997, radius=2.1003, center=<0.2 0 1.2704>, rotation=<rot 90 0 0>}]
 dyn_lng_hair_03_l [{shapeType=Capsule, length=4.9848, radius=2.0152, center=<1 0.1014 0.6232>, rotation=<rot 90 0 0>}]
 dyn_lng_hair_04_l [
  {shapeType=Capsule, length=4.5, radius=1.25, center=<1.5 0 0.5985>, rotation=<rot 90 0 0>}
  {shapeType=Capsule, length=4.25, radius=1.25, center=<2.6589 4.2876 0.5342>, rotation=<rot 90 0 0>}
  {shapeType=Capsule, length=4.5, radius=1, center=<1.8866 -3.915 -0.9432>, rotation=<rot 90 0 0>}
 ]
 dyn_lng_hair_05_l [
  {shapeType=Capsule, length=3.25, radius=1.25, center=<2 0 0>, rotation=<rot 90 0 0>}
  {shapeType=Capsule, length=5, radius=1, center=<1.224 3.7577 -0.1608>, rotation=<rot -89.0671 -179.9832 179.9899>}
  {shapeType=Capsule, length=4.5, radius=1, center=<3.0378 -3.3782 -0.711>, rotation=<rot -81.5963 -179.991 179.9983>}
 ]
 dyn_rng_hair_01_r [{shapeType=Capsule, length=4.25, radius=2.75, center=<0.1 0 0>, rotation=<rot 90 0 0>}]
 dyn_rng_hair_02_r [{shapeType=Capsule, length=4.75, radius=2.25, center=<0.2 0 0>, rotation=<rot 90 0 0>}]
 dyn_rng_hair_03_r [{shapeType=Capsule, length=6.25, radius=2.25, center=<1.9964 0 0>, rotation=<rot 90 0 0>}]
 dyn_rng_hair_04_r [
  {shapeType=Capsule, length=3.75, radius=2, center=<0.55 0 0>, rotation=<rot 90 0 0>}
  {shapeType=Capsule, length=3.5, radius=1.25, center=<2.8558 2.6593 -0.0643>, rotation=<rot 90 0 -6.8949>}
  {shapeType=Capsule, length=3.5, radius=1.5, center=<1.9845 -2.1422 -0.8303>, rotation=<rot 90 180 -165.0366>}
 ]
 dyn_rng_hair_05_r [
  {shapeType=Capsule, length=3.5, radius=2.25, center=<0.7 0 0>, rotation=<rot 90 0 0>}
  {shapeType=Capsule, length=3.75, radius=1.75, center=<1.2241 3.2377 -0.1609>, rotation=<rot -89.0671 -179.9832 179.9899>}
  {shapeType=Capsule, length=3.75, radius=1.75, center=<1.6851 -3.3784 -0.5113>, rotation=<rot -81.5963 -179.991 179.9983>}
 ]
 dyn_rng_hair_06_r [
  {shapeType=Capsule, length=4.25, radius=2.25, center=<1.6113 1.8598 -1.3581>, rotation=<rot 56.8574 11.2949 -14.2977>}
  {shapeType=Capsule, length=4.75, radius=2.25, center=<0.9152 -3.1768 -0.0172>, rotation=<rot 56.8571 11.2959 -14.2988>}
 ]
 head [
  {shapeType=Sphere, radius=8.4074, shapeName=head_sphere, center=<-0.8139 2.8266 0>}
  {shapeType=Capsule, length=9, radius=6.5, center=<5.3611 1.9755 0>, rotation=<rot 0 -176.6373 179.9999>}
 ]
 neck_01 [
  {shapeType=Capsule, length=4.25, radius=8.25, center=<0.2882 0.5436 0>, rotation=<rot 90 90 2.0263>}
  {shapeType=Capsule, length=23.0395, radius=2.15, shapeName=neck_01_capsule, center=<5.127 3.8413 -8.7835>, rotation=<rot 0 -19.9998 10.0015>}
  {shapeType=Capsule, length=21.1096, radius=2.65, shapeName=neck_01_capsule, center=<2.3755 5.8494 7.5892>, rotation=<rot 0 -19.9999 5.0008>}
 ]
 spine_05 [
  {shapeType=Capsule, length=18.6627, radius=10.5, center=<-8.1879 2.3508 -14.0107>}
  {shapeType=Capsule, length=15.8201, radius=10.5, center=<-8.1879 2.3508 14.0107>}
 ]
 upperarm_l [{shapeType=Capsule, length=22.75, radius=7.25, center=<12 0 0>, rotation=<rot 90 0 0>}]
 upperarm_r [{shapeType=Capsule, length=29.25, radius=7.75, center=<-12 0 0>, rotation=<rot 90 0 0>}]
}
constraints [
 {bone1Name=dyn_hood_frt, bone2Name=head, swing1LimitDegrees=8, swing2LimitDegrees=8, twistLimitDegrees=1}
 {bone1Name=dyn_hood_bck, bone2Name=head, swing1LimitDegrees=15, swing2LimitDegrees=15, twistLimitDegrees=7}
 {
  bone1Name=head
  bone2Name=neck_01
  swing1LimitDegrees=45
  swing1Motion=Free
  swing2LimitDegrees=45
  swing2Motion=Free
  twistLimitDegrees=45
  twistMotion=Free
 }
 {bone1Name=dyn_hood_mouth, bone2Name=neck_01, swing1LimitDegrees=11, swing2LimitDegrees=11, twistLimitDegrees=4}
 {
  bone1Name=dyn_driver_lng_hair_l
  bone2Name=head
  swing1LimitDegrees=45
  swing1Motion=Free
  swing2LimitDegrees=45
  swing2Motion=Free
  twistLimitDegrees=45
  twistMotion=Free
 }
 {bone1Name=dyn_lng_hair_01_l, bone2Name=dyn_driver_lng_hair_l, swing1LimitDegrees=6, swing2LimitDegrees=15, twistLimitDegrees=5}
 {bone1Name=dyn_lng_hair_02_l, bone2Name=dyn_lng_hair_01_l, swing1LimitDegrees=8, swing2LimitDegrees=25.2857, twistLimitDegrees=5}
 {bone1Name=dyn_lng_hair_03_l, bone2Name=dyn_lng_hair_02_l, swing1LimitDegrees=11, swing2LimitDegrees=23.4286, twistLimitDegrees=5}
 {
  bone1Name=neck_01
  bone2Name=spine_05
  swing1LimitDegrees=45
  swing1Motion=Free
  swing2LimitDegrees=45
  swing2Motion=Free
  twistLimitDegrees=45
  twistMotion=Free
 }
 {bone1Name=dyn_lng_hair_04_l, bone2Name=dyn_lng_hair_03_l, swing1LimitDegrees=13, swing2LimitDegrees=35, twistLimitDegrees=4}
 {bone1Name=dyn_lng_hair_05_l, bone2Name=dyn_lng_hair_04_l, swing1LimitDegrees=16, swing2LimitDegrees=35, twistLimitDegrees=5}
 {
  bone1Name=clavicle_l
  bone2Name=spine_05
  swing1LimitDegrees=45
  swing1Motion=Free
  swing2LimitDegrees=45
  swing2Motion=Free
  twistLimitDegrees=45
  twistMotion=Free
 }
 {
  bone1Name=clavicle_r
  bone2Name=spine_05
  swing1LimitDegrees=45
  swing1Motion=Free
  swing2LimitDegrees=45
  swing2Motion=Free
  twistLimitDegrees=45
  twistMotion=Free
 }
 {bone1Name=dyn_frnt_hair_01_r, bone2Name=head, swing1LimitDegrees=10, swing2LimitDegrees=10, twistLimitDegrees=4}
 {bone1Name=dyn_frnt_hair_02_r, bone2Name=dyn_frnt_hair_01_r, swing1LimitDegrees=15, swing2LimitDegrees=15, twistLimitDegrees=4}
 {bone1Name=dyn_frnt_hair_01_l, bone2Name=head, swing1LimitDegrees=10, swing2LimitDegrees=10, twistLimitDegrees=4}
 {bone1Name=dyn_frnt_hair_02_l, bone2Name=dyn_frnt_hair_01_l, swing1LimitDegrees=15, swing2LimitDegrees=15, twistLimitDegrees=4}
 {bone1Name=dyn_rng_hair_01_r, bone2Name=dyn_driver_rng_hair_r, swing1LimitDegrees=6, swing2LimitDegrees=15, twistLimitDegrees=5}
 {bone1Name=dyn_rng_hair_02_r, bone2Name=dyn_rng_hair_01_r, swing1LimitDegrees=8, swing2LimitDegrees=25.2857, twistLimitDegrees=5}
 {bone1Name=dyn_rng_hair_03_r, bone2Name=dyn_rng_hair_02_r, swing1LimitDegrees=11, swing2LimitDegrees=23.4286, twistLimitDegrees=5}
 {bone1Name=dyn_rng_hair_04_r, bone2Name=dyn_rng_hair_03_r, swing1LimitDegrees=13, swing2LimitDegrees=35, twistLimitDegrees=4}
 {bone1Name=dyn_rng_hair_05_r, bone2Name=dyn_rng_hair_04_r, swing1LimitDegrees=16, swing2LimitDegrees=35, twistLimitDegrees=5}
 {
  bone1Name=dyn_rng_hair_06_r
  bone2Name=dyn_rng_hair_05_r
  swing1LimitDegrees=45
  swing1Motion=Free
  swing2LimitDegrees=45
  swing2Motion=Free
  twistLimitDegrees=45
  twistMotion=Free
 }
 {
  bone1Name=dyn_driver_rng_hair_r
  bone2Name=head
  swing1LimitDegrees=45
  swing1Motion=Free
  swing2LimitDegrees=45
  swing2Motion=Free
  twistLimitDegrees=45
  twistMotion=Free
 }
 {
  bone1Name=upperarm_r
  bone2Name=clavicle_r
  swing1LimitDegrees=45
  swing1Motion=Free
  swing2LimitDegrees=45
  swing2Motion=Free
  twistLimitDegrees=45
  twistMotion=Free
 }
 {
  bone1Name=upperarm_l
  bone2Name=clavicle_l
  swing1LimitDegrees=45
  swing1Motion=Free
  swing2LimitDegrees=45
  swing2Motion=Free
  twistLimitDegrees=45
  twistMotion=Free
 }
]
