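# Post-process style AnimBP for the Rainbow Hat female face accessory, animating its hair and hood bones.
# It caches a pose copied from parent and body meshes, offsetting hair driver bones by head and arm rotation.
# RigidBody simulations, selected by skydiving and front-end state, feed hood constraints and rifle-aiming corrections, with an AnimDynamics chain when rigid bodies are disabled.
# The EventGraph derives blend alphas and gravity override; the targeting_alpha bone chain is unconnected.
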AnimBlueprint /BRCosmetics/Characters/Player/Female/Medium/Bodies/F_MED_Rainbow_Hat/Meshes/Parts/F_MED_Rainbow_Hat_FaceAcc_AnimBP
target_skeleton=&/F_MED_Rainbow_Hat_FaceAcc_Skeleton
graphs [
 AnimGraph {nodes [
   Anim Dynamics - Bone: dyn_rng_hair_06_r(Alpha=1, ExternalForce="0,0,0") {
    angularDampingOverride=0.9999
    angularSpringConstant=100
    bAngularSpring=true
    bGravityOverrideInSimSpace=true
    bOverrideAngularDamping=true
    simulationSpace=RootRelative
    alphaBoolBlend {blendOption=Linear}
    boundBone {boneName=dyn_rng_hair_06_r}
    componentAppliedLinearAccClamp=<3500 3500 3500>
    componentLinearAccScale=<1.2 1.2 1.2>
    componentLinearVelScale=<1.2 1.2 1.2>
    physicsBodyDefinitions [{
      sphereCollisionRadius=0
      boundBone {boneName=dyn_rng_hair_06_r}
      boxExtents=<20 10 10>
      constraintSetup {angularLimitsMax=<1 5 15>, angularLimitsMin=<-1 -30 -15>, angularTarget=<0 0 0>}
      localJointOffset=<2 0 0>
     }]
    relativeSpaceBone {boneName=head}
   }
   n1 = "Transform (Modify) Bone - Bone: dyn_rng_hair_04_r"(Alpha=Get arm_rot_high_r() -> (arm_rot_high_r:double), Translation="-0.000,-1,-0.000", Rotation="11.880,0.001,-0.002", Scale="1,1,1") {
    rotationMode=BMM_Additive
    rotationSpace=BCS_BoneSpace
    translationMode=BMM_Additive
    translationSpace=BCS_BoneSpace
    alphaBoolBlend {blendOption=Linear}
    alphaScaleBiasClamp {bInterpResult=true}
    boneToModify {boneName=dyn_rng_hair_04_r}
    rotation=<rot 11.88 0.001 -0.002>
    translation {y=-1}
   }
   n2 = "Transform (Modify) Bone - Bone: dyn_lng_hair_02_l"(
    ComponentPose=@n1
    Alpha=Get arm_rot_high_l() -> (arm_rot_high_l:double)
    Translation="-0.320,1.241,-0.037"
    Rotation="2.502,14.532,0.071"
    Scale="1,1,1"
   ) {
    rotationMode=BMM_Additive
    rotationSpace=BCS_BoneSpace
    translationMode=BMM_Additive
    translationSpace=BCS_BoneSpace
    alphaBoolBlend {blendOption=Linear}
    alphaScaleBiasClamp {bInterpResult=true}
    boneToModify {boneName=dyn_lng_hair_02_l}
    rotation=<rot 2.502 14.532 0.071>
    translation=<-0.32 1.241 -0.037>
   }
   n3 = Get targeting_alpha() -> (targeting_alpha:double)
   n4 = "Transform (Modify) Bone - Bone: dyn_rng_hair_02_r"(ComponentPose=@n2, Alpha=@n3, Translation="1.413,-3.756,-0.149", Rotation="12.744,-13.273,2.143", Scale="1,1,1") {
    rotationMode=BMM_Additive
    rotationSpace=BCS_BoneSpace
    translationMode=BMM_Additive
    translationSpace=BCS_BoneSpace
    alphaBoolBlend {blendOption=Linear}
    alphaScaleBiasClamp {bInterpResult=true}
    boneToModify {boneName=dyn_rng_hair_02_r}
    rotation=<rot 12.744 -13.2733 2.143>
    translation=<1.413 -3.756 -0.149>
   }
   "Transform (Modify) Bone - Bone: dyn_rng_hair_03_r"(ComponentPose=@n4, Alpha=@n3, Translation="0,0,0", Rotation="-1.443,15.542,-0.000", Scale="1,1,1") {
    rotationMode=BMM_Additive
    rotationSpace=BCS_BoneSpace
    translationMode=BMM_Additive
    translationSpace=BCS_BoneSpace
    alphaBoolBlend {blendOption=Linear}
    alphaScaleBiasClamp {bInterpResult=true}
    boneToModify {boneName=dyn_rng_hair_03_r}
    rotation=<rot -1.443 15.5422 0>
   }
   n5 = RigidBody!(ComponentPose=Local To Component(LocalPose=Use cached pose 'AnimCache'()), Alpha, OverrideWorldGravity:Vector="0,0,-950", ExternalForce) -> (Pose) {
    bClampLinearTranslationLimitToRefPose=true
    bOverrideWorldGravity=true
    overridePhysicsAsset=&/F_MED_Rainbow_Hat_FE_FaceAcc_Physics
    simulationSpace=BaseBoneSpace
    alphaBoolBlend {blendOption=Linear}
    baseBoneRef {boneName=spine_03}
    componentAppliedLinearAccClamp=<1100 1100 1100>
    componentLinearAccScale=<0.5 0.5 0.6>
    componentLinearVelScale=<0.5 0.5 0.6>
    overrideWorldGravity {z=-950}
   }
   n6 = RigidBody!(ComponentPose=Local To Component(LocalPose=Use cached pose 'AnimCache'()), Alpha, ExternalForce) -> (Pose) {
    bClampLinearTranslationLimitToRefPose=true
    bOverrideWorldGravity=true
    overridePhysicsAsset=&/F_MED_Rainbow_Hat_FaceAcc_Physics
    simulationSpace=BaseBoneSpace
    alphaBoolBlend {blendOption=Linear}
    baseBoneRef {boneName=spine_03}
    componentAppliedLinearAccClamp=<1300 1300 1300>
    componentLinearAccScale=<0.5 0.5 0.6>
    componentLinearVelScale=<0.5 0.5 0.6>
    overrideWorldGravity {z=-950}
   }
   n7 = Blend Poses by bool(BlendPose_0=Component To Local(ComponentPose=@n5), BlendPose_1=Component To Local(ComponentPose=@n6), bActiveValue=Get bIsInFrontEnd()) {blendType=Linear}
   n8 = RigidBody!(
    ComponentPose=Local To Component(LocalPose=Use cached pose 'AnimCache'())
    Alpha
    OverrideWorldGravity:Vector=Get Gravity_Override() -> (Gravity_Override:Vector)
    ExternalForce
   ) -> (Pose) {
    bClampLinearTranslationLimitToRefPose=true
    bOverrideWorldGravity=true
    overridePhysicsAsset=&F_MED_Rainbow_Hat_FE_FaceAcc_Physics
    simulationSpace=BaseBoneSpace
    alphaBoolBlend {blendOption=Linear}
    baseBoneRef {boneName=spine_03}
    componentAppliedLinearAccClamp=<1100 1100 1100>
    componentLinearAccScale=<0.5 0.5 0.6>
    componentLinearVelScale=<0.5 0.5 0.6>
    overrideWorldGravity {z=-950}
   }
   n9 = Component To Local(ComponentPose=@n8)
   n10 = Get bIsSkyDiving()
   n11 = "Constraint - dyn_hood_r (P)"(
    ComponentPose=Local To Component(LocalPose=Blend Poses by bool(BlendPose_0=@n9, BlendPose_1=@n7, bActiveValue=@n10) {blendType=Linear})
    Alpha=1
    ConstraintWeights_0=1
   ) {
    alphaBoolBlend {blendOption=Linear}
    boneToModify {boneName=dyn_hood_r}
    constraintSetup [{targetBone {boneName=dyn_parent_rng_hair_r}}]
    constraintWeights [1]
   }
   n12 = "Constraint - dyn_hood_l (P)"(ComponentPose=@n11, Alpha=1, ConstraintWeights_0=1) {
    alphaBoolBlend {blendOption=Linear}
    boneToModify {boneName=dyn_hood_l}
    constraintSetup [{targetBone {boneName=dyn_parent_lng_hair_l}}]
    constraintWeights [1]
   }
   n13 = "Constraint - dyn_hood_frt (PP)"!(ComponentPose=@n12, Alpha=0.8, ConstraintWeights_0=0.5, ConstraintWeights_1:float=0.5) -> (Pose) {
    alphaBoolBlend {blendOption=Linear}
    boneToModify {boneName=dyn_hood_frt}
    constraintSetup [{targetBone {boneName=dyn_parent_lng_hair_l}}, {targetBone {boneName=dyn_parent_rng_hair_r}}]
    constraintWeights [1,1]
   }
   n14 = "Constraint - dyn_hood_bck (PP)"!(ComponentPose=@n13, Alpha=0.5, ConstraintWeights_0=0.5, ConstraintWeights_1:float=0.5) -> (Pose) {
    alphaBoolBlend {blendOption=Linear}
    boneToModify {boneName=dyn_hood_bck}
    constraintSetup [{targetBone {boneName=dyn_parent_lng_hair_l}}, {targetBone {boneName=dyn_parent_rng_hair_r}}]
    constraintWeights [1,1]
   }
   n15 = Get ADS_Up_Crouching() -> (ADS_Up_Crouching:double)
   n16 = "Transform (Modify) Bone - Bone: dyn_hood_bck"!(ComponentPose=@n14, Alpha=@n15, Translation="2,0,-2", Scale="0.8,1.000,1.000") -> (Pose) {
    alpha=0
    rotationMode=BMM_Additive
    rotationSpace=BCS_BoneSpace
    scaleMode=BMM_Additive
    scaleSpace=BCS_BoneSpace
    translationMode=BMM_Additive
    translationSpace=BCS_BoneSpace
    alphaBoolBlend {blendOption=Linear}
    alphaScaleBiasClamp {bInterpResult=true, interpSpeedDecreasing=45, interpSpeedIncreasing=45}
    boneToModify {boneName=dyn_hood_bck}
    scale {x=0.8}
    translation {x=2, z=-2}
   }
   n17 = "Transform (Modify) Bone - Bone: dyn_hood_r"!(ComponentPose=@n16, Alpha=@n15, Translation="2,1,0") -> (Pose) {
    alpha=0
    rotationMode=BMM_Additive
    rotationSpace=BCS_BoneSpace
    translationMode=BMM_Additive
    translationSpace=BCS_BoneSpace
    alphaBoolBlend {blendOption=Linear}
    alphaScaleBiasClamp {bInterpResult=true, interpSpeedDecreasing=15, interpSpeedIncreasing=15}
    boneToModify {boneName=dyn_hood_r}
    translation=<2 1>
   }
   n18 = Get ADS() -> (ADS:double)
   n19 = "Constraint - dyn_rng_hair_02_r (P)"(ComponentPose=@n17, Alpha=@n18, ConstraintWeights_0=1) {
    alpha=0
    alphaBoolBlend {blendOption=Linear}
    alphaScaleBiasClamp {bClampResult=true, clampMax=0.8}
    boneToModify {boneName=dyn_rng_hair_02_r}
    constraintSetup [{targetBone {boneName=clavicle_r}}]
    constraintWeights [1]
   }
   n20 = "Constraint - dyn_rng_hair_04_r (P)"(ComponentPose=@n19, Alpha=@n18, ConstraintWeights_0=1) {
    alpha=0
    alphaBoolBlend {blendOption=Linear}
    alphaScaleBiasClamp {bClampResult=true}
    boneToModify {boneName=dyn_rng_hair_04_r}
    constraintSetup [{targetBone {boneName=clavicle_r}}]
    constraintWeights [1]
   }
   n21 = "Constraint - dyn_rng_hair_05_r (P)"(ComponentPose=@n20, Alpha=@n18, ConstraintWeights_0=1) {
    alpha=0
    alphaBoolBlend {blendOption=Linear}
    alphaScaleBiasClamp {bClampResult=true, clampMax=0.5}
    boneToModify {boneName=dyn_rng_hair_05_r}
    constraintSetup [{targetBone {boneName=clavicle_r}}]
    constraintWeights [1]
   }
   n22 = "Transform (Modify) Bone - Bone: dyn_rng_hair_01_r"(ComponentPose=@n21, Alpha=@n18, Translation="0.000,0,0.5", Rotation="0,0,0", Scale="1,1,1") {
    alpha=0
    rotationMode=BMM_Additive
    rotationSpace=BCS_BoneSpace
    scaleMode=BMM_Additive
    scaleSpace=BCS_BoneSpace
    translationMode=BMM_Additive
    translationSpace=BCS_BoneSpace
    alphaBoolBlend {blendOption=Linear}
    alphaScaleBiasClamp {bInterpResult=true, interpSpeedIncreasing=15}
    boneToModify {boneName=dyn_rng_hair_01_r}
    translation {z=0.5}
   }
   n23 = "Transform (Modify) Bone - Bone: dyn_rng_hair_03_r"(ComponentPose=@n22, Alpha=@n18, Translation="0,0,0", Rotation="-20,20,45", Scale="1,1,1") {
    alpha=0
    rotationMode=BMM_Additive
    rotationSpace=BCS_BoneSpace
    scaleMode=BMM_Additive
    scaleSpace=BCS_BoneSpace
    translationMode=BMM_Additive
    translationSpace=BCS_BoneSpace
    alphaBoolBlend {blendOption=Linear}
    alphaScaleBiasClamp {bInterpResult=true, interpSpeedIncreasing=15}
    boneToModify {boneName=dyn_rng_hair_03_r}
    rotation=<rot -20 20 45>
   }
   n24 = "Transform (Modify) Bone - Bone: dyn_rng_hair_04_r"!(ComponentPose=@n23, Alpha=@n18, Translation="0.000,0,-0.5", Rotation="-12,12,10") -> (Pose) {
    alpha=0
    rotationMode=BMM_Additive
    rotationSpace=BCS_BoneSpace
    scaleMode=BMM_Additive
    scaleSpace=BCS_BoneSpace
    translationMode=BMM_Additive
    translationSpace=BCS_BoneSpace
    alphaBoolBlend {blendOption=Linear}
    alphaScaleBiasClamp {bInterpResult=true, interpSpeedIncreasing=15}
    boneToModify {boneName=dyn_rng_hair_04_r}
    rotation=<rot -12 12 10>
    scale {z=0}
    translation {z=-0.5}
   }
   n25 = Component To Local(ComponentPose=@n24)
   n26 = "Constraint - dyn_lng_hair_01_l (PP)"!(ComponentPose=Local To Component(LocalPose=Use cached pose 'AnimCache'()), Alpha=1, ConstraintWeights_0=0.4, ConstraintWeights_1:float=0.6) -> (Pose) {
    alphaBoolBlend {blendOption=Linear}
    boneToModify {boneName=dyn_lng_hair_01_l}
    constraintSetup [{targetBone {boneName=clavicle_l}}, {targetBone {boneName=dyn_hood_l}}]
    constraintWeights [1,1]
   }
   n27 = "Constraint - dyn_rng_hair_01_r (PP)"!(ComponentPose=@n26, Alpha=1, ConstraintWeights_0=0.4, ConstraintWeights_1:float=0.6) -> (Pose) {
    alphaBoolBlend {blendOption=Linear}
    boneToModify {boneName=dyn_rng_hair_01_r}
    constraintSetup [{targetBone {boneName=clavicle_r}}, {targetBone {boneName=dyn_hood_r}}]
    constraintWeights [1,1]
   }
   n28 = "Constraint - dyn_lng_hair_04_l (P)"(ComponentPose=@n27, Alpha=0.3, ConstraintWeights_0=1) {
    alphaBoolBlend {blendOption=Linear}
    boneToModify {boneName=dyn_lng_hair_04_l}
    constraintSetup [{targetBone {boneName=spine_05}}]
    constraintWeights [1]
   }
   n29 = "Constraint - dyn_rng_hair_04_r (P)"(ComponentPose=@n28, Alpha=0.3, ConstraintWeights_0=1) {
    alphaBoolBlend {blendOption=Linear}
    boneToModify {boneName=dyn_rng_hair_04_r}
    constraintSetup [{targetBone {boneName=spine_05}}]
    constraintWeights [1]
   }
   n30 = "Transform (Modify) Bone - Bone: dyn_lng_hair_02_l"(ComponentPose=@n29, Alpha=1, Translation="0,0,0", Rotation="10,0,0", Scale="1,1,1") {
    alpha=0
    rotationMode=BMM_Additive
    rotationSpace=BCS_BoneSpace
    scaleMode=BMM_Additive
    scaleSpace=BCS_BoneSpace
    translationMode=BMM_Additive
    translationSpace=BCS_BoneSpace
    alphaBoolBlend {blendOption=Linear}
    alphaScaleBiasClamp {interpSpeedDecreasing=45, interpSpeedIncreasing=45}
    boneToModify {boneName=dyn_lng_hair_02_l}
    rotation {pitch=10}
   }
   n31 = Anim Dynamics - Bone: dyn_frnt_hair_01_l(ComponentPose=@n30, Alpha=1, ExternalForce="0,0,0") {
    angularDampingOverride=0.9999
    angularSpringConstant=100
    bAngularSpring=true
    bGravityOverrideInSimSpace=true
    bOverrideAngularDamping=true
    simulationSpace=RootRelative
    alphaBoolBlend {blendOption=Linear}
    boundBone {boneName=dyn_frnt_hair_01_l}
    componentAppliedLinearAccClamp=<3500 3500 3500>
    componentLinearAccScale=<1.2 1.2 1.2>
    componentLinearVelScale=<1.2 1.2 1.2>
    physicsBodyDefinitions [{
      sphereCollisionRadius=0
      boundBone {boneName=dyn_frnt_hair_01_l}
      boxExtents=<20 10 10>
      constraintSetup {angularLimitsMax=<1 1 15>, angularLimitsMin=<-1 -30 -15>, angularTarget=<0 0 0>}
      localJointOffset=<2 0 0>
     }]
    relativeSpaceBone {boneName=head}
   }
   n32 = Anim Dynamics - Bone: dyn_frnt_hair_02_l(ComponentPose=@n31, Alpha=1, ExternalForce="0,0,0") {
    angularDampingOverride=0.9999
    angularSpringConstant=100
    bAngularSpring=true
    bGravityOverrideInSimSpace=true
    bOverrideAngularDamping=true
    simulationSpace=RootRelative
    alphaBoolBlend {blendOption=Linear}
    boundBone {boneName=dyn_frnt_hair_02_l}
    componentAppliedLinearAccClamp=<3500 3500 3500>
    componentLinearAccScale=<1.2 1.2 1.2>
    componentLinearVelScale=<1.2 1.2 1.2>
    physicsBodyDefinitions [{
      sphereCollisionRadius=0
      boundBone {boneName=dyn_frnt_hair_02_l}
      boxExtents=<20 10 10>
      constraintSetup {angularLimitsMax=<1 1 15>, angularLimitsMin=<-1 -30 -15>, angularTarget=<0 0 0>}
      localJointOffset=<2 0 0>
     }]
    relativeSpaceBone {boneName=head}
    sphericalLimits [
     {limitRadius=20, drivingBone {boneName=spine_04}, sphereLocalOffset=<-5 1 -12>}
     {limitRadius=20, drivingBone {boneName=spine_03}, sphereLocalOffset=<0 0 0>}
    ]
   }
   n33 = Anim Dynamics - Bone: dyn_frnt_hair_01_r(ComponentPose=@n32, Alpha=1, ExternalForce="0,0,0") {
    angularDampingOverride=0.9999
    angularSpringConstant=100
    bAngularSpring=true
    bGravityOverrideInSimSpace=true
    bOverrideAngularDamping=true
    simulationSpace=RootRelative
    alphaBoolBlend {blendOption=Linear}
    boundBone {boneName=dyn_frnt_hair_01_r}
    componentAppliedLinearAccClamp=<3500 3500 3500>
    componentLinearAccScale=<1.2 1.2 1.2>
    componentLinearVelScale=<1.2 1.2 1.2>
    physicsBodyDefinitions [{
      sphereCollisionRadius=0
      boundBone {boneName=dyn_frnt_hair_01_r}
      boxExtents=<20 10 10>
      constraintSetup {angularLimitsMax=<1 1 15>, angularLimitsMin=<-1 -30 -15>, angularTarget=<0 0 0>}
      localJointOffset=<2 0 0>
     }]
    relativeSpaceBone {boneName=head}
   }
   n34 = Anim Dynamics - Bone: dyn_frnt_hair_02_r(ComponentPose=@n33, Alpha=1, ExternalForce="0,0,0") {
    angularDampingOverride=0.9999
    angularSpringConstant=100
    bAngularSpring=true
    bGravityOverrideInSimSpace=true
    bOverrideAngularDamping=true
    simulationSpace=RootRelative
    alphaBoolBlend {blendOption=Linear}
    boundBone {boneName=dyn_frnt_hair_02_r}
    componentAppliedLinearAccClamp=<3500 3500 3500>
    componentLinearAccScale=<1.2 1.2 1.2>
    componentLinearVelScale=<1.2 1.2 1.2>
    physicsBodyDefinitions [{
      sphereCollisionRadius=0
      boundBone {boneName=dyn_frnt_hair_02_r}
      boxExtents=<20 10 10>
      constraintSetup {angularLimitsMax=<1 1 15>, angularLimitsMin=<-1 -30 -15>, angularTarget=<0 0 0>}
      localJointOffset=<2 0 0>
     }]
    relativeSpaceBone {boneName=head}
    sphericalLimits [
     {limitRadius=15, drivingBone {boneName=spine_04}, sphereLocalOffset=<0 2 10>}
     {limitRadius=5, drivingBone {boneName=spine_05}, sphereLocalOffset=<0 0 0>}
    ]
   }
   n35 = Anim Dynamics - Bone: dyn_lng_hair_02_l(ComponentPose=@n34, Alpha=1, ExternalForce="0,0,0") {
    angularDampingOverride=0.9999
    angularSpringConstant=100
    bAngularSpring=true
    bGravityOverrideInSimSpace=true
    bOverrideAngularDamping=true
    simulationSpace=RootRelative
    alphaBoolBlend {blendOption=Linear}
    boundBone {boneName=dyn_lng_hair_02_l}
    componentAppliedLinearAccClamp=<3500 3500 3500>
    componentLinearAccScale=<1.2 1.2 1.2>
    componentLinearVelScale=<1.2 1.2 1.2>
    physicsBodyDefinitions [{
      sphereCollisionRadius=0
      boundBone {boneName=dyn_lng_hair_02_l}
      boxExtents=<20 10 10>
      constraintSetup {angularLimitsMax=<1 1 15>, angularLimitsMin=<-1 -30 -15>, angularTarget=<0 0 0>}
      localJointOffset=<2 0 0>
     }]
    relativeSpaceBone {boneName=head}
   }
   n36 = Anim Dynamics - Bone: dyn_lng_hair_03_l(ComponentPose=@n35, Alpha=1, ExternalForce="0,0,0") {
    angularDampingOverride=0.9999
    angularSpringConstant=100
    bAngularSpring=true
    bGravityOverrideInSimSpace=true
    bOverrideAngularDamping=true
    bUseSphericalLimits=true
    simulationSpace=RootRelative
    alphaBoolBlend {blendOption=Linear}
    boundBone {boneName=dyn_lng_hair_03_l}
    componentAppliedLinearAccClamp=<3500 3500 3500>
    componentLinearAccScale=<1.2 1.2 1.2>
    componentLinearVelScale=<1.2 1.2 1.2>
    physicsBodyDefinitions [{
      sphereCollisionRadius=0
      boundBone {boneName=dyn_lng_hair_03_l}
      boxExtents=<20 10 10>
      constraintSetup {angularLimitsMax=<1 1 15>, angularLimitsMin=<-1 -30 -15>, angularTarget=<0 0 0>}
      localJointOffset=<2 0 0>
     }]
    relativeSpaceBone {boneName=head}
    sphericalLimits [{limitRadius=10, drivingBone {boneName=clavicle_l}, sphereLocalOffset=<0 0 0>}]
   }
   n37 = Anim Dynamics - Bone: dyn_lng_hair_04_l(ComponentPose=@n36, Alpha=1, ExternalForce="0,0,0") {
    angularDampingOverride=0.9999
    angularSpringConstant=100
    bAngularSpring=true
    bGravityOverrideInSimSpace=true
    bOverrideAngularDamping=true
    simulationSpace=RootRelative
    alphaBoolBlend {blendOption=Linear}
    boundBone {boneName=dyn_lng_hair_04_l}
    componentAppliedLinearAccClamp=<3500 3500 3500>
    componentLinearAccScale=<1.2 1.2 1.2>
    componentLinearVelScale=<1.2 1.2 1.2>
    physicsBodyDefinitions [{
      sphereCollisionRadius=0
      boundBone {boneName=dyn_lng_hair_04_l}
      boxExtents=<20 10 10>
      constraintSetup {angularLimitsMax=<1 1 15>, angularLimitsMin=<-1 -30 -15>, angularTarget=<0 0 0>}
      localJointOffset=<2 0 0>
     }]
    relativeSpaceBone {boneName=head}
    sphericalLimits [
     {limitRadius=20, drivingBone {boneName=spine_04}, sphereLocalOffset=<-5 1 -12>}
     {limitRadius=20, drivingBone {boneName=spine_03}, sphereLocalOffset=<0 0 0>}
    ]
   }
   n38 = Anim Dynamics - Bone: dyn_lng_hair_05_l(ComponentPose=@n37, Alpha=1, ExternalForce="0,0,0") {
    angularDampingOverride=0.9999
    angularSpringConstant=100
    bAngularSpring=true
    bGravityOverrideInSimSpace=true
    bOverrideAngularDamping=true
    bUseSphericalLimits=true
    simulationSpace=RootRelative
    alphaBoolBlend {blendOption=Linear}
    boundBone {boneName=dyn_lng_hair_05_l}
    componentAppliedLinearAccClamp=<3500 3500 3500>
    componentLinearAccScale=<1.2 1.2 1.2>
    componentLinearVelScale=<1.2 1.2 1.2>
    physicsBodyDefinitions [{
      sphereCollisionRadius=0
      boundBone {boneName=dyn_lng_hair_05_l}
      boxExtents=<20 10 10>
      constraintSetup {angularLimitsMax=<1 1 15>, angularLimitsMin=<-1 -30 -15>, angularTarget=<0 0 0>}
      localJointOffset=<2 0 0>
     }]
    relativeSpaceBone {boneName=head}
    sphericalLimits [
     {limitRadius=15, drivingBone {boneName=spine_04}, sphereLocalOffset=<0 1 -10>}
     {limitRadius=12, drivingBone {boneName=spine_04}, sphereLocalOffset=<0 0 0>}
    ]
   }
   n39 = Anim Dynamics - Bone: dyn_rng_hair_02_r(ComponentPose=@n38, Alpha=1, ExternalForce="0,0,0") {
    angularDampingOverride=0.9999
    angularSpringConstant=100
    bAngularSpring=true
    bGravityOverrideInSimSpace=true
    bOverrideAngularDamping=true
    simulationSpace=RootRelative
    alphaBoolBlend {blendOption=Linear}
    boundBone {boneName=dyn_rng_hair_02_r}
    componentAppliedLinearAccClamp=<3500 3500 3500>
    componentLinearAccScale=<1.2 1.2 1.2>
    componentLinearVelScale=<1.2 1.2 1.2>
    physicsBodyDefinitions [{
      sphereCollisionRadius=0
      boundBone {boneName=dyn_rng_hair_02_r}
      boxExtents=<20 10 10>
      constraintSetup {angularLimitsMax=<1 1 15>, angularLimitsMin=<-1 -30 -15>, angularTarget=<0 0 0>}
      localJointOffset=<2 0 0>
     }]
    relativeSpaceBone {boneName=head}
   }
   n40 = Anim Dynamics - Bone: dyn_rng_hair_03_r(ComponentPose=@n39, Alpha=1, ExternalForce="0,0,0") {
    angularDampingOverride=0.9999
    angularSpringConstant=100
    bAngularSpring=true
    bGravityOverrideInSimSpace=true
    bOverrideAngularDamping=true
    bUseSphericalLimits=true
    simulationSpace=RootRelative
    alphaBoolBlend {blendOption=Linear}
    boundBone {boneName=dyn_rng_hair_03_r}
    componentAppliedLinearAccClamp=<3500 3500 3500>
    componentLinearAccScale=<1.2 1.2 1.2>
    componentLinearVelScale=<1.2 1.2 1.2>
    physicsBodyDefinitions [{
      sphereCollisionRadius=0
      boundBone {boneName=dyn_rng_hair_03_r}
      boxExtents=<20 10 10>
      constraintSetup {angularLimitsMax=<1 1 15>, angularLimitsMin=<-1 -30 -15>, angularTarget=<0 0 0>}
      localJointOffset=<2 0 0>
     }]
    relativeSpaceBone {boneName=head}
    sphericalLimits [{limitRadius=10, drivingBone {boneName=clavicle_r}, sphereLocalOffset=<0 0 0>}]
   }
   n41 = Anim Dynamics - Bone: dyn_rng_hair_04_r(ComponentPose=@n40, Alpha=1, ExternalForce="0,0,0") {
    angularDampingOverride=0.9999
    angularSpringConstant=100
    bAngularSpring=true
    bGravityOverrideInSimSpace=true
    bOverrideAngularDamping=true
    simulationSpace=RootRelative
    alphaBoolBlend {blendOption=Linear}
    boundBone {boneName=dyn_rng_hair_04_r}
    componentAppliedLinearAccClamp=<3500 3500 3500>
    componentLinearAccScale=<1.2 1.2 1.2>
    componentLinearVelScale=<1.2 1.2 1.2>
    physicsBodyDefinitions [{
      sphereCollisionRadius=0
      boundBone {boneName=dyn_rng_hair_04_r}
      boxExtents=<20 10 10>
      constraintSetup {angularLimitsMax=<1 1 15>, angularLimitsMin=<-1 -30 -15>, angularTarget=<0 0 0>}
      localJointOffset=<2 0 0>
     }]
    relativeSpaceBone {boneName=head}
    sphericalLimits [
     {limitRadius=15, drivingBone {boneName=spine_04}, sphereLocalOffset=<0 2 10>}
     {limitRadius=5, drivingBone {boneName=spine_05}, sphereLocalOffset=<0 0 0>}
    ]
   }
   n42 = Anim Dynamics - Bone: dyn_rng_hair_05_r(ComponentPose=@n41, Alpha=1, ExternalForce="0,0,0") {
    angularDampingOverride=0.9999
    angularSpringConstant=100
    bAngularSpring=true
    bGravityOverrideInSimSpace=true
    bOverrideAngularDamping=true
    bUseSphericalLimits=true
    simulationSpace=RootRelative
    alphaBoolBlend {blendOption=Linear}
    boundBone {boneName=dyn_rng_hair_05_r}
    componentAppliedLinearAccClamp=<3500 3500 3500>
    componentLinearAccScale=<1.2 1.2 1.2>
    componentLinearVelScale=<1.2 1.2 1.2>
    physicsBodyDefinitions [{
      sphereCollisionRadius=0
      boundBone {boneName=dyn_rng_hair_05_r}
      boxExtents=<20 10 10>
      constraintSetup {angularLimitsMax=<1 1 15>, angularLimitsMin=<-1 -30 -15>, angularTarget=<0 0 0>}
      localJointOffset=<2 0 0>
     }]
    relativeSpaceBone {boneName=head}
    sphericalLimits [
     {limitRadius=15, drivingBone {boneName=spine_04}, sphereLocalOffset=<0 1 10>}
     {limitRadius=5, drivingBone {boneName=spine_04}, sphereLocalOffset=<0 0 0>}
    ]
   }
   n43 = Component To Local(ComponentPose=@n42)
   n44 = Get bRigidBodyEnabled()
   n45 = "Constraint - dyn_lng_hair_03_l (P)"(
    ComponentPose=Local To Component(LocalPose=Blend Poses by bool(BlendPose_0=@n25, BlendPose_1=@n43, bActiveValue=@n44) {blendType=Linear})
    Alpha=1
    ConstraintWeights_0=1
   ) {
    alpha=0
    alphaBoolBlend {blendOption=Linear}
    alphaScaleBiasClamp {bClampResult=true, bInterpResult=true, clampMax=0.5}
    boneToModify {boneName=dyn_lng_hair_03_l}
    constraintSetup [{targetBone {boneName=spine_05}}]
    constraintWeights [1]
   }
   n46 = "Constraint - dyn_rng_hair_03_r (P)"(ComponentPose=@n45, Alpha=1, ConstraintWeights_0=1) {
    alpha=0
    alphaBoolBlend {blendOption=Linear}
    alphaScaleBiasClamp {bClampResult=true, bInterpResult=true, clampMax=0.5}
    boneToModify {boneName=dyn_rng_hair_03_r}
    constraintSetup [{targetBone {boneName=spine_05}}]
    constraintWeights [1]
   }
   Output Pose(Result=Component To Local(ComponentPose=@n46)) {layerGroup=None}
   n47 = Component To Local(ComponentPose=Local To Component(LocalPose=Copy Pose From Mesh() {bUseAttachedParent=true}))
   n48 = Copy Pose From Mesh(SourceMeshComponent=Get Body_Mesh()) {bUseAttachedParent=true}
   n49 = "Transform (Modify) Bone - Bone: dyn_parent_rng_hair_r"!(
    ComponentPose=Local To Component(LocalPose=Layered blend per bone(BasePose=@n47, BlendPoses_0=@n48) {layerSetup [{branchFilters [{boneName=upperarm_l}, {boneName=upperarm_r}]}]})
    Alpha=1
    Translation="-1.500,0.000,0"
   ) -> (Pose) {
    rotationMode=BMM_Additive
    rotationSpace=BCS_BoneSpace
    scaleMode=BMM_Additive
    scaleSpace=BCS_BoneSpace
    translationMode=BMM_Additive
    translationSpace=BCS_BoneSpace
    alphaBoolBlend {blendOption=Linear}
    alphaScaleBiasClamp {bInterpResult=true, interpSpeedDecreasing=12, interpSpeedIncreasing=12}
    boneToModify {boneName=dyn_parent_rng_hair_r}
    translation {x=-1.5}
   }
   n50 = "Transform (Modify) Bone - Bone: dyn_driver_rng_hair_r"!(ComponentPose=@n49, Alpha=1, Translation="0,0.5,-0.000", Rotation="0.000,10,0.000") -> (Pose) {
    rotationMode=BMM_Additive
    rotationSpace=BCS_BoneSpace
    scaleMode=BMM_Additive
    scaleSpace=BCS_BoneSpace
    translationMode=BMM_Additive
    translationSpace=BCS_BoneSpace
    alphaBoolBlend {blendOption=Linear}
    alphaScaleBiasClamp {bInterpResult=true, interpSpeedDecreasing=12, interpSpeedIncreasing=12}
    boneToModify {boneName=dyn_driver_rng_hair_r}
    rotation {yaw=10}
    translation {y=0.5}
   }
   n51 = "Transform (Modify) Bone - Bone: dyn_driver_lng_hair_l"!(ComponentPose=@n50, Alpha=1, Translation="-1,-0.5,0.4", Rotation="0.000,-10,0.000") -> (Pose) {
    rotationMode=BMM_Additive
    rotationSpace=BCS_BoneSpace
    translationMode=BMM_Additive
    translationSpace=BCS_BoneSpace
    alphaBoolBlend {blendOption=Linear}
    alphaScaleBiasClamp {bInterpResult=true, interpSpeedDecreasing=15, interpSpeedIncreasing=12}
    boneToModify {boneName=dyn_driver_lng_hair_l}
    rotation {yaw=-10}
    translation=<-1 -0.5 0.4>
   }
   n52 = Get Head Up() -> (Head Up:double)
   n53 = "Transform (Modify) Bone - Bone: dyn_parent_rng_hair_r"(ComponentPose=@n51, Alpha=@n52, Translation="0,1.2,0.000") {
    alpha=0
    rotationMode=BMM_Additive
    rotationSpace=BCS_BoneSpace
    scaleMode=BMM_Additive
    scaleSpace=BCS_BoneSpace
    translationMode=BMM_Additive
    translationSpace=BCS_BoneSpace
    alphaBoolBlend {blendOption=Linear}
    alphaScaleBiasClamp {interpSpeedDecreasing=12, interpSpeedIncreasing=12}
    boneToModify {boneName=dyn_parent_rng_hair_r}
    translation {y=1.2}
   }
   n54 = "Transform (Modify) Bone - Bone: dyn_driver_lng_hair_l"(ComponentPose=@n53, Alpha=@n52, Translation="2,0,2", Rotation="0,0,0") {
    alpha=0
    rotationMode=BMM_Additive
    rotationSpace=BCS_BoneSpace
    scaleMode=BMM_Additive
    scaleSpace=BCS_BoneSpace
    translationMode=BMM_Additive
    translationSpace=BCS_BoneSpace
    alphaBoolBlend {blendOption=Linear}
    alphaScaleBiasClamp {interpSpeedDecreasing=12, interpSpeedIncreasing=12}
    boneToModify {boneName=dyn_driver_lng_hair_l}
    translation {x=2, z=2}
   }
   n55 = "Transform (Modify) Bone - Bone: dyn_driver_rng_hair_r"(ComponentPose=@n54, Alpha=Get Head Down() -> (Head Down:double), Translation="0,0,2", Rotation="0,-10,0") {
    alpha=0
    rotationMode=BMM_Additive
    rotationSpace=BCS_BoneSpace
    scaleMode=BMM_Additive
    scaleSpace=BCS_BoneSpace
    translationMode=BMM_Additive
    translationSpace=BCS_BoneSpace
    alphaBoolBlend {blendOption=Linear}
    alphaScaleBiasClamp {interpSpeedDecreasing=12, interpSpeedIncreasing=12}
    boneToModify {boneName=dyn_driver_rng_hair_r}
    rotation {yaw=-10}
    translation {z=2}
   }
   n56 = Get Head Left() -> (Head Left:double)
   n57 = "Transform (Modify) Bone - Bone: dyn_parent_rng_hair_r"(ComponentPose=@n55, Alpha=@n56, Translation="-1,-1,0.000") {
    alpha=0
    rotationMode=BMM_Additive
    rotationSpace=BCS_BoneSpace
    scaleMode=BMM_Additive
    scaleSpace=BCS_BoneSpace
    translationMode=BMM_Additive
    translationSpace=BCS_BoneSpace
    alphaBoolBlend {blendOption=Linear}
    alphaScaleBiasClamp {interpSpeedDecreasing=12, interpSpeedIncreasing=12}
    boneToModify {boneName=dyn_parent_rng_hair_r}
    translation=<-1 -1>
   }
   n58 = "Transform (Modify) Bone - Bone: dyn_parent_lng_hair_l"!(ComponentPose=@n57, Alpha=@n56, Translation="1,-1,0.000") -> (Pose) {
    alpha=0
    rotationMode=BMM_Additive
    rotationSpace=BCS_BoneSpace
    scaleMode=BMM_Additive
    scaleSpace=BCS_BoneSpace
    translationMode=BMM_Additive
    translationSpace=BCS_BoneSpace
    alphaBoolBlend {blendOption=Linear}
    alphaScaleBiasClamp {interpSpeedDecreasing=12, interpSpeedIncreasing=12}
    boneToModify {boneName=dyn_parent_lng_hair_l}
    translation=<1 -1>
   }
   n59 = "Transform (Modify) Bone - Bone: dyn_driver_rng_hair_r"(ComponentPose=@n58, Alpha=@n56, Translation="0,-0.8,2", Rotation="0,0,0") {
    alpha=0
    rotationMode=BMM_Additive
    rotationSpace=BCS_BoneSpace
    scaleMode=BMM_Additive
    scaleSpace=BCS_BoneSpace
    translationMode=BMM_Additive
    translationSpace=BCS_BoneSpace
    alphaBoolBlend {blendOption=Linear}
    alphaScaleBiasClamp {interpSpeedDecreasing=12, interpSpeedIncreasing=12}
    boneToModify {boneName=dyn_driver_rng_hair_r}
    translation {y=-0.8, z=2}
   }
   n60 = "Transform (Modify) Bone - Bone: dyn_driver_rng_hair_r"!(ComponentPose=@n59, Alpha=Get arm_rot_high_r(), Translation="-0.000,-5.25,-0.000", Rotation="0,0,0", Scale:Vector="1,1,1") -> (Pose) {
    rotationMode=BMM_Additive
    rotationSpace=BCS_BoneSpace
    translationMode=BMM_Additive
    translationSpace=BCS_BoneSpace
    alphaBoolBlend {blendOption=Linear}
    alphaScaleBiasClamp {bInterpResult=true, interpSpeedDecreasing=12, interpSpeedIncreasing=12}
    boneToModify {boneName=dyn_driver_rng_hair_r}
    translation {y=-5.25}
   }
   n61 = "Transform (Modify) Bone - Bone: dyn_driver_lng_hair_l"!(ComponentPose=@n60, Alpha=Get arm_rot_high_l(), Translation="0.000,0,3", Rotation="-35.599,-0.000,0.000", Scale:Vector="1,1,1") -> (Pose) {
    rotationMode=BMM_Additive
    rotationSpace=BCS_BoneSpace
    translationMode=BMM_Additive
    translationSpace=BCS_BoneSpace
    alphaBoolBlend {blendOption=Linear}
    alphaScaleBiasClamp {bInterpResult=true, interpSpeedDecreasing=15, interpSpeedIncreasing=12}
    boneToModify {boneName=dyn_driver_lng_hair_l}
    rotation {pitch=-35.599}
    translation {z=3}
   }
   Save cached pose 'AnimCache'(Pose=Component To Local(ComponentPose=@n61)) {cacheName=AnimCache}
   Get Gravity_Override()
  ]}
 EventGraph {nodes [
   Event BlueprintUpdateAnimation()
   > Get!(execute:Exec) -> (then:Exec, else:Exec, Body_Mesh:Skeletal Mesh Component Object Reference)
   > Get!(execute) -> (then, else, Skel_Mesh:Skeletal Mesh Component Object Reference)
   > n1 = Sequence!(execute) -> (then_0, then_1, then_2:Exec, then_3:Exec, then_4:Exec)
   n2 = GetOwnerSkeletalMeshComponent()
   n3 = BreakRotator(InRot=BreakTransform(InTransform=GetDeltaTransformFromRefPose(self=@n2, BoneName=head, BaseName=spine_05)).Rotation)
   Set Head Up(execute=@n1.then_1, Head Up:double=MapRangeClamped(Value=@n3.Yaw, InRangeA=-9, InRangeB=-40)) -> (Output_Get:double)
   > Set Head Down(Head Down:double=MapRangeClamped(Value=@n3.Yaw, InRangeA=-9, InRangeB=16)) -> (Output_Get:double)
   > Set Head Left(Head Left:double=MapRangeClamped(Value=@n3.Pitch, InRangeA=0.9, InRangeB=3)) -> (Output_Get:double)
   n4 = Branch(execute=@n1.then_2, Condition=Get isAimingRifle() -> (isAimingRifle:bool))
   n5 = BreakTransform(InTransform=GetDeltaTransformFromRefPose(self=GetOwnerSkeletalMeshComponent(), BoneName=neck_01, BaseName=spine_05))
   > Set ADS(ADS:double=MapRangeClamped(Value=BreakRotator(InRot=@n5.Rotation).Yaw, InRangeA=5, InRangeB=9)) -> (Output_Get:double)
   Set ADS(execute=@n4.else, ADS=0)
   n6 = Cast To FortPlayerAnimInstance(execute=@n1.then_3, Object=GetAnimInstance(self=GetOwnerSkeletalMeshComponent()))
   > Set Rifle(Rifle:double=0) -> (Output_Get:double)
   > n7 = Switch on EFortWeaponCoreAnimation!(execute, Selection:EFortWeaponCoreAnimation Enum=Get WeaponCoreAnim(self=@n6)) -> (
    Melee:Exec
    Pistol:Exec
    Shotgun:Exec
    PaperBlueprint:Exec
    Rifle:Exec
    MeleeOneHand:Exec
    MachinePistol:Exec
    RocketLauncher:Exec
    GrenadeLauncher:Exec
    GoingCommando:Exec
    AssaultRifle:Exec
    TacticalShotgun:Exec
    SniperRifle:Exec
    TrapPlacement:Exec
    ShoulderLauncher:Exec
    AbilityDecoTool:Exec
    Crossbow:Exec
    C4:Exec
    RemoteControl:Exec
    DualWield:Exec
    AR_BullPup:Exec
    AR_ForwardGrip:Exec
    MedPackPaddles:Exec
    SMG_P90:Exec
    AR_DrumGun:Exec
    Consumable_Small:Exec
    Consumable_Large:Exec
    Balloon:Exec
    MountedTurret:Exec
    CreativeTool:Exec
    ExplosiveBow:Exec
    AshtonIndigo:Exec
    AshtonChicago:Exec
    MeleeDualWield:Exec
    Unarmed:Exec
   )
   Branch(execute=[@n7.AssaultRifle, @n7.Crossbow, @n7.GrenadeLauncher, @n7.Rifle, @n7.RocketLauncher, @n7.Shotgun, @n7.SniperRifle, @n7.TacticalShotgun])
   > Set Rifle(Rifle=1)
   > n8 = Branch(Condition=Get bIsTargeting())
   > Set isAimingRifle(isAimingRifle:bool=true) -> (Output_Get:bool)
   Set isAimingRifle(execute=@n8.else, isAimingRifle=false)
   n9 = Gravity Override State Function DEPRECATED(
    execute=@n1.then_4
    CharacterPartAnimInstance=Self-Reference()
    GameTimeInSeconds=GetGameTimeInSeconds()
    GravityOverrideParams=Get Gravity Override Params()
   )
   > Set Gravity_Override(Gravity_Override:Vector=@n9) -> (Output_Get:Vector)
   n10 = Get Skel_Mesh()
   Set arm_rot_high_r(
    execute=@n1.then_0
    arm_rot_high_r:double=GetAlphaFromReferenceTransformDelta(SkeletalMeshComponent=@n10, BoneName=upperarm_r, BaseName=clavicle_r, InputRange=(X=30,Y=90), TwistAxis="0,0,1")
   ) -> (Output_Get:double)
   > Set arm_rot_high_l(arm_rot_high_l:double=GetAlphaFromReferenceTransformDelta(SkeletalMeshComponent=@n10, BoneName=upperarm_l, BaseName=clavicle_l, InputRange=(X=20,Y=90), TwistAxis="0,0,1")) -> (Output_Get:double)
   > n11 = Branch(Condition=AND Boolean(A=Get bIsTargeting(), B=Get bIsCrouching()))
   n12 = BreakTransform(InTransform=GetDeltaTransformFromRefPose(self=GetOwnerSkeletalMeshComponent(), BoneName=head, BaseName=spine_05))
   > Set ADS_Up_Crouching(ADS_Up_Crouching:double=MapRangeClamped(Value=BreakRotator(InRot=@n12.Rotation).Yaw, InRangeA=-18, InRangeB=-35)) -> (Output_Get:double)
   Set ADS_Up_Crouching(execute=@n11.else, ADS_Up_Crouching=0)
   Event BlueprintInitializeAnimation()
   > Set Skel_Mesh(Skel_Mesh=GetOwnerSkeletalMeshComponent())
   > n13 = Cast To FortPlayerPawn(Object=GetOwningActor())
   > Is Valid(InputObject=@n13)
   > Set Pawn(Pawn:Fort Player Pawn Object Reference=@n13) -> (Output_Get:Fort Player Pawn Object Reference)
   > Set Body_Mesh(Body_Mesh=GetSkeletalMeshForPartType(self=@n13))
   Event OnEndTargeting() {is=title K2Node_Event}
   > Set targeting_alpha(targeting_alpha:double=0) -> (Output_Get:double)
   Event OnBeginTargeting() {is=title K2Node_Event}
   > Set targeting_alpha(targeting_alpha=1)
   TryGetPawnOwner()
   Is Valid(InputObject=GetOwnerSkeletalMeshComponent())
   Get bIsTargeting()
  ]}
]
variables [
 ADS {type=number}
 ADS_Up_Crouching {type=number}
 arm_rot_high_l {type=number}
 arm_rot_high_r {type=number}
 Body_Mesh {type=SkeletalMeshComponent}
 Gravity Override Params {type=GravityOverrideParamsStruct}
 Gravity_Override {type=Vector}
 Head Down {type=number}
 Head Left {type=number}
 Head Up {type=number}
 isAimingRifle {type=boolean}
 Pawn {type=FortPlayerPawn}
 Rifle {type=number}
 Skel_Mesh {type=SkeletalMeshComponent}
 targeting_alpha {type=number}
]
